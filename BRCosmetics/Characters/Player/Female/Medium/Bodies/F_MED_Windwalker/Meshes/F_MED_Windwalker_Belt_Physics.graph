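# Physics asset for the belt of the female medium Windwalker player cosmetic, setting up two pouch chains for physical motion.
# Each chain has a tiny root capsule constrained to the pelvis capsule with free rotation, plus a long capsule for the pouch.
# Pouch capsules have uneven swing limits and very tight twist relative to their roots.
# A small spine_01 box also joins the pelvis with free rotation.
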PhysicsAsset /BRCosmetics/Characters/Player/Female/Medium/Bodies/F_MED_Windwalker/Meshes/F_MED_Windwalker_Belt_Physics
bodies {
 dyn_pouch_01 [{shapeType=Capsule, length=0.505, radius=0.505, shapeName=dyn_pouch_01_capsule, rotation=<rot 0 -90 0>}]
 dyn_pouch_02 [{shapeType=Capsule, length=17.2963, radius=2.8854, shapeName=dyn_pouch_02_capsule, center=<9.0743 0.0001 0.0162>, rotation=<rot 90 0 0>}]
 dyn_pouch_03 [{shapeType=Capsule, length=0.505, radius=0.505, shapeName=dyn_pouch_03_capsule, rotation=<rot 0 -90 0>}]
 dyn_pouch_04 [{shapeType=Capsule, length=15.4035, radius=3.0204, shapeName=dyn_pouch_04_capsule, center=<8.7377 0 0.0164>, rotation=<rot 90 0 0>}]
 pelvis [{shapeType=Capsule, length=6, radius=10.25, center=<-4.2317 -1.547 -0.0002>}]
 spine_01 [{shapeType=Box, extentX=1.01, extentY=1.01, extentZ=1.01, rotation=<rot 0 -90 0>}]
}
constraints [
 {bone1Name=dyn_pouch_02, bone2Name=dyn_pouch_01, swing1LimitDegrees=45, swing2LimitDegrees=10, twistLimitDegrees=1}
 {bone1Name=dyn_pouch_04, bone2Name=dyn_pouch_03, swing1LimitDegrees=45, swing2LimitDegrees=10, twistLimitDegrees=1}
 {
  bone1Name=spine_01
  bone2Name=pelvis
  swing1LimitDegrees=45
  swing1Motion=Free
  swing2LimitDegrees=45
  swing2Motion=Free
  twistLimitDegrees=45
  twistMotion=Free
 }
 {
  bone1Name=dyn_pouch_03
  bone2Name=pelvis
  swing1LimitDegrees=45
  swing1Motion=Free
  swing2LimitDegrees=45
  swing2Motion=Free
  twistLimitDegrees=45
  twistMotion=Free
 }
 {
  bone1Name=dyn_pouch_01
  bone2Name=pelvis
  swing1LimitDegrees=45
  swing1Motion=Free
  swing2LimitDegrees=45
  swing2Motion=Free
  twistLimitDegrees=45
  twistMotion=Free
 }
]
