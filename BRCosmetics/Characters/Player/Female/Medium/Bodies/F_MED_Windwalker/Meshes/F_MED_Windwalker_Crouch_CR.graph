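# Corrective Control Rig for the female medium Windwalker body (named Crouch) that shifts the dyn_strap bone as the right arm moves.
# It converts upperarm_r's rotation relative to clavicle_r to ZXY Euler and remaps Z over a clamped range into a driver value.
# That value scales a small X translation and Y rotation fed into a Modify Transforms node on the strap.
# A left-clavicle transform read and a negate node feed nothing.
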ControlRigBlueprint /BRCosmetics/Characters/Player/Female/Medium/Bodies/F_MED_Windwalker/Meshes/F_MED_Windwalker_Crouch_CR
graphs [
 RigVMFunctionLibrary {links [], nodes []}
 RigVMModel {nodes [
   n1 = &RigUnit_BeginExecution()
   n2 = &RigVMFunction_MathQuaternionToEuler(
    RotationOrder=ZXY
    Value=&RigUnit_GetRelativeTransformForItem(Child=(Type=Bone,Name="upperarm_r"), Parent=(Type=Bone,Name="clavicle_r"), bChildInitial=false).RelativeTransform.Rotation
   )
   n3 = &RigVMFunction_MathFloatRemap(SourceMaximum=130, SourceMinimum=103, bClamp=true, Value=@n2.Result.Z)
   &RigUnit_ModifyTransforms(
    ItemToModify=((Item=(Type=Bone,Name="dyn_strap"),Transform=(Rotation=(X=0,Y=0,Z=0,W=1),Translation=(X=0,Y=0,Z=0),Scale3D=(X=1,Y=1,Z=1))))
    Weight=1
    ItemToModify.0.Transform.Translation.X=&RigVMFunction_MathFloatMul(B=-5, A=@n3.Result) {template="Multiply::Execute(in A,in B,out Result)"}.Result
    ItemToModify.0.Transform.Translation.Y=&RigVMFunction_MathFloatMul(B=0, A=@n3.Result) {template="Multiply::Execute(in A,in B,out Result)"}.Result
    ItemToModify.0.Transform.Translation.Z=&RigVMFunction_MathFloatMul(B=0, A=@n3.Result) {template="Multiply::Execute(in A,in B,out Result)"}.Result
    ItemToModify.0.Transform.Rotation=&RigVMFunction_MathQuaternionFromEuler(
     Euler=(X=0,Y=30,Z=-30)
     Euler.Z=&RigVMFunction_MathFloatMul(B=0, A=@n3.Result) {template="Multiply::Execute(in A,in B,out Result)"}.Result
     Euler.Y=&RigVMFunction_MathFloatMul(B=-6, A=@n3.Result) {template="Multiply::Execute(in A,in B,out Result)"}.Result
    ).Result
    ExecutePin=@n1.ExecutePin
   )
   &RigUnit_GetRelativeBoneTransform(Bone=clavicle_l, Space=spine_05)
   /Script/RigVM.RigVMFunction_MathFloatNegate(Value=@n2.Result.Y) {is=struct RigVMUnitNode, template="Negate::Execute(in Value,out Result)"}
  ]}
]
hierarchy [
 root {loc=<0 0 0>, rot=<0 0 0 1>}
 pelvis {parent=root, loc=<0 -0.5301 98.0571>, rot=<0.0127 0.707 -0.0127 -0.707>}
 spine_01 {parent=pelvis, loc=<6.1208 -0.9328 0>, rot=<0 0 0.034 -0.9994>}
 spine_02 {parent=spine_01, loc=<6.5996 -0.2095 0>, rot=<0 0 0.0238 -0.9997>}
 spine_03 {parent=spine_02, loc=<6.4914 -0.1237 0>, rot=<0 0 0.0149 -0.9999>}
 spine_04 {parent=spine_03, loc=<6.9598 0.0794 0>, rot=<0 0 -0.0132 -0.9999>}
 spine_05 {parent=spine_04, loc=<6.9987 0 0>, rot=<0 0 -0.0015 -1>}
 clavicle_l {parent=spine_05, loc=<5.4831 -0.1053 -3.2946>, rot=<0.0465 -0.7289 -0.009 -0.6829>}
 upperarm_l {parent=clavicle_l, loc=<9.8658 -1.2136 -0.4105>, rot=<0.0164 -0.3245 -0.0269 -0.9454>}
 lowerarm_l {parent=upperarm_l, loc=<26.465 -0.8345 0.259>, rot=<0.0036 0.0161 -0.2298 -0.9731>}
 hand_l {parent=lowerarm_l, loc=<22.7355 -0.3203 -0.6106>, rot=<-0.6784 -0.0223 0.0735 -0.7307>}
 index_metacarpal_l {parent=hand_l, loc=<1.6535 0.0183 -2.96>, rot=<0.0158 -0.0124 -0.0173 -0.9996>}
 index_01_l {parent=index_metacarpal_l, loc=<8.3008 0.0357 -0.0524>, rot=<-0.0044 0.0485 0.0924 -0.9945>}
 index_02_l {parent=index_01_l, loc=<3.3587 -0.3004 0.0452>, rot=<-0.002 0.0027 0.3225 -0.9466>}
 index_03_l {parent=index_02_l, loc=<2.5578 -0.123 -0.0252>, rot=<-0.002 0.0067 0.2558 -0.9667>}
 middle_metacarpal_l {parent=hand_l, loc=<2.1856 -0.1725 -0.5984>, rot=<0.006 0.0168 -0.0329 -0.9993>}
 middle_01_l {parent=middle_metacarpal_l, loc=<7.3275 -0.2447 0.0129>, rot=<-0.0208 0.0042 0.0858 -0.9961>}
 middle_02_l {parent=middle_01_l, loc=<4.0123 0.1401 0.1175>, rot=<-0.0218 0.0258 0.3512 -0.9357>}
 middle_03_l {parent=middle_02_l, loc=<2.6077 0.0534 0.1373>, rot=<0.0082 0.0188 0.1708 -0.9851>}
 pinky_metacarpal_l {parent=hand_l, loc=<1.7616 -0.9753 3.3814>, rot=<-0.0397 0.0814 -0.069 -0.9935>}
 pinky_01_l {parent=pinky_metacarpal_l, loc=<7.1709 -0.4831 -0.1094>, rot=<-0.0654 -0.0035 0.2242 -0.9723>}
 pinky_02_l {parent=pinky_01_l, loc=<2.8274 0.0893 -0.1043>, rot=<0.0037 0.0011 0.2854 -0.9584>}
 pinky_03_l {parent=pinky_02_l, loc=<2.2046 0.0307 -0.0078>, rot=<-0.0002 0.0098 0.1445 -0.9895>}
 ring_metacarpal_l {parent=hand_l, loc=<2.2074 -0.3625 1.3638>, rot=<-0.0284 0.0722 -0.0458 -0.9959>}
 ring_01_l {parent=ring_metacarpal_l, loc=<7.0295 -0.6197 -0.1446>, rot=<-0.0489 -0.0252 0.1892 -0.9804>}
 ring_02_l {parent=ring_01_l, loc=<3.6174 0.4647 -0.1027>, rot=<0 0 0.2806 -0.9598>}
 ring_03_l {parent=ring_02_l, loc=<2.4152 0.114 0.0199>, rot=<0 0 0.1189 -0.9929>}
 thumb_01_l {parent=hand_l, loc=<1.3118 -0.3191 -2.459>, rot=<0.6546 -0.4042 0.0664 -0.6354>}
 thumb_02_l {parent=thumb_01_l, loc=<5.9336 -0.0448 0.0401>, rot=<-0.0037 0.0213 0.298 -0.9543>}
 thumb_03_l {parent=thumb_02_l, loc=<2.2223 0.3682 -0.0907>, rot=<0.0118 -0.013 0.0066 -0.9998>}
 lowerarm_twist_01_l {parent=lowerarm_l, loc=<18.316 0 0>, rot=<0 0 0 1>}
 lowerarm_twist_02_l {parent=lowerarm_l, loc=<10.6707 0 0>, rot=<0 0 0 1>}
 upperarm_twist_01_l {parent=upperarm_l, loc=<7.2135 -0.1134 0>, rot=<0 0 0 1>}
 upperarm_twist_02_l {parent=upperarm_l, loc=<19.7998 -0.1134 0>, rot=<0 0 0 1>}
 dyn_shash_mstr {parent=clavicle_l, loc=<4.1923 -2.4826 5.5654>, rot=<0.58 -0.4257 0.2889 -0.6316>}
 dyn_sash_01 {parent=dyn_shash_mstr, loc=<5.6991 2.29 -0.1656>, rot=<-0.1086 -0.0426 -0.428 -0.8962>}
 dyn_sash_02 {parent=dyn_sash_01, loc=<6.4677 1.2116 -0.1568>, rot=<-0.0586 -0.0403 -0.2 -0.9772>}
 dyn_sash_03 {parent=dyn_sash_02, loc=<8.1788 -0.4328 0.2459>, rot=<0 0 0 1>}
 dyn_sash_04 {parent=dyn_sash_03, loc=<8.2812 0.3975 0.1247>, rot=<0 0 0 1>}
 dyn_sash_05 {parent=dyn_sash_04, loc=<10.126 0.1304 0.1994>, rot=<0 0 0 1>}
 dyn_sash_06 {parent=dyn_sash_05, loc=<11.6705 0.094 0.3155>, rot=<0 0 0 1>}
 dyn_sash_07 {parent=dyn_sash_06, loc=<10.8794 0.2378 -0.0281>, rot=<0 0 0 1>}
 dyn_sash_08 {parent=dyn_sash_07, loc=<11.9225 0.4371 0.1984>, rot=<0 0 0 1>}
 clavicle_r {parent=spine_05, loc=<5.4831 -0.1053 3.2946>, rot=<0.7289 0.0465 -0.6829 0.009>}
 upperarm_r {parent=clavicle_r, loc=<-9.8658 1.2136 0.4104>, rot=<0.0164 -0.3245 -0.0269 -0.9454>}
 lowerarm_r {parent=upperarm_r, loc=<-26.4651 0.8345 -0.259>, rot=<0.0036 0.0161 -0.2298 -0.9731>}
 hand_r {parent=lowerarm_r, loc=<-22.7356 0.3203 0.6106>, rot=<-0.6784 -0.0223 0.0735 -0.7307>}
 index_metacarpal_r {parent=hand_r, loc=<-1.6534 -0.0181 2.9601>, rot=<0.0158 -0.0124 -0.0173 -0.9996>}
 index_01_r {parent=index_metacarpal_r, loc=<-8.3008 -0.0357 0.0524>, rot=<-0.0044 0.0485 0.0924 -0.9945>}
 index_02_r {parent=index_01_r, loc=<-3.3589 0.3003 -0.0453>, rot=<-0.002 0.0027 0.3225 -0.9466>}
 index_03_r {parent=index_02_r, loc=<-2.5578 0.123 0.0252>, rot=<-0.002 0.0067 0.2558 -0.9667>}
 middle_metacarpal_r {parent=hand_r, loc=<-2.1856 0.1726 0.5984>, rot=<0.006 0.0168 -0.0329 -0.9993>}
 middle_01_r {parent=middle_metacarpal_r, loc=<-7.3273 0.2447 -0.0129>, rot=<-0.0208 0.0042 0.0858 -0.9961>}
 middle_02_r {parent=middle_01_r, loc=<-4.0124 -0.14 -0.1175>, rot=<-0.0218 0.0258 0.3512 -0.9357>}
 middle_03_r {parent=middle_02_r, loc=<-2.6077 -0.0533 -0.1373>, rot=<0.0082 0.0188 0.1708 -0.9851>}
 pinky_metacarpal_r {parent=hand_r, loc=<-1.7617 0.9753 -3.3814>, rot=<-0.0397 0.0814 -0.069 -0.9935>}
 pinky_01_r {parent=pinky_metacarpal_r, loc=<-7.1707 0.4832 0.1094>, rot=<-0.0654 -0.0035 0.2242 -0.9723>}
 pinky_02_r {parent=pinky_01_r, loc=<-2.8274 -0.0892 0.1043>, rot=<0.0037 0.0011 0.2854 -0.9584>}
 pinky_03_r {parent=pinky_02_r, loc=<-2.2045 -0.031 0.0078>, rot=<-0.0002 0.0098 0.1445 -0.9895>}
 ring_metacarpal_r {parent=hand_r, loc=<-2.2073 0.3627 -1.3638>, rot=<-0.0284 0.0722 -0.0458 -0.9959>}
 ring_01_r {parent=ring_metacarpal_r, loc=<-7.0297 0.6197 0.1446>, rot=<-0.0489 -0.0252 0.1892 -0.9804>}
 ring_02_r {parent=ring_01_r, loc=<-3.6171 -0.4648 0.1027>, rot=<0 0 0.2806 -0.9598>}
 ring_03_r {parent=ring_02_r, loc=<-2.4153 -0.1141 -0.0199>, rot=<0 0 0.1189 -0.9929>}
 thumb_01_r {parent=hand_r, loc=<-1.3119 0.3193 2.459>, rot=<0.6546 -0.4042 0.0664 -0.6354>}
 thumb_02_r {parent=thumb_01_r, loc=<-5.9336 0.0447 -0.0401>, rot=<-0.0037 0.0213 0.298 -0.9543>}
 thumb_03_r {parent=thumb_02_r, loc=<-2.2223 -0.3682 0.0907>, rot=<0.0118 -0.013 0.0066 -0.9998>}
 lowerarm_twist_01_r {parent=lowerarm_r, loc=<-18.3161 0 0>, rot=<0 0 0 1>}
 lowerarm_twist_02_r {parent=lowerarm_r, loc=<-10.6707 0 0>, rot=<0 0 0 1>}
 upperarm_twist_01_r {parent=upperarm_r, loc=<-7.2135 0.1134 0>, rot=<0 0 0 1>}
 upperarm_twist_02_r {parent=upperarm_r, loc=<-19.7998 0.1134 0>, rot=<0 0 0 1>}
 neck_01 {parent=spine_05, loc=<11.3138 0.1529 0>, rot=<0 0 -0.2003 -0.9797>}
 neck_02 {parent=neck_01, loc=<4.5627 0 0>, rot=<0 0 0.0504 -0.9987>}
 head {parent=neck_02, loc=<5.3914 0 0>, rot=<0 0 0.115 -0.9934>}
 dyn_strap {parent=spine_05, loc=<10.6931 0.4735 9.6134>, rot=<0.3919 -0.5609 0.6759 0.2738>}
 dyn_pouch_05 {parent=spine_01, loc=<2.6713 5.1333 -11.9108>, rot=<-0.2024 0.4733 0.8409 0.1672>}
 dyn_pouch_06 {parent=dyn_pouch_05, loc=<0 0 0>, rot=<0 0 0 -1>}
 thigh_l {parent=pelvis, loc=<-1.0855 -0.2109 -8.7053>, rot=<0.0009 0.0355 0.0155 -0.9993>}
 calf_l {parent=thigh_l, loc=<-42.5284 -0.0019 -0.7297>, rot=<0.0006 -0.0161 -0.0391 -0.9991>}
 calf_twist_01_l {parent=calf_l, loc=<-35.8483 -1.9197 -0.5794>, rot=<0.0237 0.0095 -0.0035 -0.9997>}
 calf_twist_02_l {parent=calf_l, loc=<-19.1546 -1.7943 -0.2638>, rot=<0.0237 0.0095 -0.0035 -0.9997>}
 foot_l {parent=calf_l, loc=<-46.7401 -2.5014 -0.592>, rot=<0.0427 -0.0238 0.0353 -0.9982>}
 ball_l {parent=foot_l, loc=<-3.5658 10.8066 -0.1544>, rot=<0.0001 -0.0001 -0.7063 -0.7079>}
 thigh_twist_01_l {parent=thigh_l, loc=<-27.952 0 0>, rot=<0.0013 0 0 -1>}
 thigh_r {parent=pelvis, loc=<-1.0853 -0.2109 8.7053>, rot=<0.0355 -0.0009 0.9993 0.0155>}
 calf_r {parent=thigh_r, loc=<42.5286 0.0019 0.7297>, rot=<0.0006 -0.0161 -0.0391 -0.9991>}
 calf_twist_01_r {parent=calf_r, loc=<35.8485 1.9197 0.5793>, rot=<0.0237 0.0095 -0.0035 -0.9997>}
 calf_twist_02_r {parent=calf_r, loc=<19.1546 1.7943 0.2638>, rot=<0.0237 0.0095 -0.0035 -0.9997>}
 foot_r {parent=calf_r, loc=<46.7403 2.5014 0.5919>, rot=<0.0427 -0.0238 0.0353 -0.9982>}
 ball_r {parent=foot_r, loc=<3.5658 -10.8066 0.1544>, rot=<0.0001 -0.0001 -0.7063 -0.7079>}
 thigh_twist_01_r {parent=thigh_r, loc=<27.952 0 0>, rot=<0.0013 0 0 -1>}
 dyn_tunic_back_r_01 {parent=pelvis, loc=<-0.3401 -7.1174 13.1471>, rot=<0.0757 -0.9926 0.0711 0.0635>}
 dyn_tunic_back_r_02 {parent=dyn_tunic_back_r_01, loc=<5.7087 0 0>, rot=<-0.0114 0.0814 -0.038 -0.9959>}
 dyn_tunic_back_mid_01 {parent=pelvis, loc=<1.5904 -11.0166 0>, rot=<0.0873 -0.9962 0 0>}
 dyn_tunic_back_mid_02 {parent=dyn_tunic_back_mid_01, loc=<7.4085 0 0>, rot=<0 0 -0.1013 -0.9949>}
 dyn_tunic_back_l_01 {parent=pelvis, loc=<-0.3401 -7.1174 -13.1471>, rot=<0.0869 -0.9582 -0.2683 -0.0471>}
 dyn_tunic_back_l_02 {parent=dyn_tunic_back_l_01, loc=<5.7087 0 0>, rot=<0.007 0.0236 -0.023 -0.9994>}
 dyn_pouch_01 {parent=pelvis, loc=<3.4365 -9.3894 11.4587>, rot=<0.0829 -0.9245 0.3719 0.0116>}
 dyn_pouch_02 {parent=dyn_pouch_01, loc=<0 0 0>, rot=<0 0 0 -1>}
 dyn_pouch_03 {parent=pelvis, loc=<4.2616 -0.0134 14.8433>, rot=<0.0445 -0.6913 0.7211 0.0111>}
 dyn_pouch_04 {parent=dyn_pouch_03, loc=<0 0 0>, rot=<0 0 0 -1>}
 dyn_tunic_front_r_03 {parent=pelvis, loc=<1.2545 8.3348 3.8999>, rot=<-0.0179 -0.9998 0 0>}
 dyn_tunic_front_r_04 {parent=dyn_tunic_front_r_03, loc=<7.2649 0 0>, rot=<0 0 0 -1>}
 dyn_tunic_front_l_03 {parent=pelvis, loc=<1.2484 8.1653 -3.8999>, rot=<-0.0179 -0.9998 0 0>}
 dyn_tunic_front_l_04 {parent=dyn_tunic_front_l_03, loc=<7.2649 0 0>, rot=<0 0 0 -1>}
 dyn_tunic_front_l_MSTR {parent=pelvis, loc=<2.7528 7.5758 -9.4424>, rot=<-0.0279 0.7029 0.7044 0.095>}
 dyn_tunic_front_l_01 {parent=dyn_tunic_front_l_MSTR, loc=<0 0 0>, rot=<0 0 0 -1>}
 dyn_tunic_front_l_02 {parent=dyn_tunic_front_l_01, loc=<9.4678 0.0608 0.3578>, rot=<-0.0004 0.0289 -0.0049 -0.9996>}
 dyn_tunic_front_r_MSTR {parent=pelvis, loc=<1.2006 6.7543 9.0666>, rot=<-0.1284 -0.7045 -0.6979 0.0107>}
 dyn_tunic_front_r_01 {parent=dyn_tunic_front_r_MSTR, loc=<0 0 0>, rot=<0 0 0 1>}
 dyn_tunic_front_r_02 {parent=dyn_tunic_front_r_01, loc=<7.6156 -0.1362 -0.8002>, rot=<0.0027 -0.0657 0.0413 -0.997>}
 dyn_tunic_side_r_MSTR {parent=pelvis, loc=<1.828 3.3911 14.1904>, rot=<0.1379 0.4315 0.8907 0.0367>}
 dyn_tunic_side_r_01 {parent=dyn_tunic_side_r_MSTR, loc=<0 0 0>, rot=<0 0 0 -1>}
 dyn_tunic_side_r_02 {parent=dyn_tunic_side_r_01, loc=<7.6156 -0.1362 -0.8002>, rot=<0.0027 -0.0657 0.0413 -0.997>}
 dyn_tunic_side_l_MSTR {parent=pelvis, loc=<1.8279 3.3912 -14.1904>, rot=<-0.0367 -0.8907 -0.4315 -0.1379>}
 dyn_tunic_side_l_03 {parent=dyn_tunic_side_l_MSTR, loc=<0 0 0>, rot=<0 0 0 -1>}
 dyn_tunic_side_l_04 {parent=dyn_tunic_side_l_03, loc=<7.6156 0.1362 -0.8002>, rot=<-0.0027 -0.0657 -0.0413 -0.997>}
]
