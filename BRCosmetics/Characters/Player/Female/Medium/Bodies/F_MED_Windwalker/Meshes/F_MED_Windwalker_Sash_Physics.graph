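# Physics asset for the sash on the female medium Windwalker body cosmetic, providing a jointed sash chain and surrounding body shapes.
# The sash hangs from the left clavicle via a master bone into eight segments, boxes then capsules, with tight twist and swing limits loosening down the chain.
# Pelvis, spine, thighs, left arm, pouch and back tunic bodies form surrounding volumes, joined by angularly free constraints.
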PhysicsAsset /BRCosmetics/Characters/Player/Female/Medium/Bodies/F_MED_Windwalker/Meshes/F_MED_Windwalker_Sash_Physics
bodies {
 clavicle_l [{shapeType=Capsule, length=8.5, radius=6, center=<4.7894 -1.7976 -1.2629>, rotation=<rot 90 0 0>}]
 dyn_pouch_02 [{shapeType=Capsule, length=25.875, radius=8.5, shapeName=dyn_pouch_02_capsule, center=<8.5 -2.9647 -0.3626>, rotation=<rot 90 0 0>}]
 dyn_sash_01 [{shapeType=Box, extentX=5, extentY=2, extentZ=5, center=<3.3441 0.298 -0.0001>, rotation=<rot 0 8.053 0>}]
 dyn_sash_02 [{shapeType=Box, extentX=10.5, extentY=2, extentZ=10, center=<2.2626 0 0>}]
 dyn_sash_03 [{shapeType=Box, extentX=10.5, extentY=2, extentZ=10, center=<2.2626 0 0>}]
 dyn_sash_04 [{shapeType=Capsule, length=8.4569, radius=3, shapeName=dyn_sash_07_capsule, center=<3.3398 0.0003 0>, rotation=<rot 0 90 0>}]
 dyn_sash_05 [{shapeType=Capsule, length=8.4569, radius=3, shapeName=dyn_sash_07_capsule, center=<3.3398 0.0003 0>, rotation=<rot 0 90 0>}]
 dyn_sash_06 [{shapeType=Capsule, length=8.4569, radius=3, shapeName=dyn_sash_07_capsule, center=<3.0332 0.0003 0>, rotation=<rot 0 90 0>}]
 dyn_sash_07 [{shapeType=Capsule, length=8.4569, radius=3, shapeName=dyn_sash_07_capsule, center=<2.8727 0.0003 0>, rotation=<rot 0 90 0>}]
 dyn_sash_08 [{shapeType=Capsule, length=7, radius=3.25, shapeName=dyn_sash_08_capsule, center=<2.3393 -0.0004 0>, rotation=<rot 0 -90 0>}]
 dyn_shash_mstr [{shapeType=Box, extentX=6.5, extentY=2, extentZ=5.5, center=<4.3817 1.2045 0>, rotation=<rot 0 20.6741 0>}]
 dyn_tunic_back_l_01 [{shapeType=Capsule, length=10.1414, radius=2.7925, shapeName=dyn_tunic_back_l_01_capsule}]
 dyn_tunic_back_l_02 [{shapeType=Capsule, length=12.9331, radius=3, shapeName=dyn_tunic_back_l_02_capsule, center=<3.7453 1.6708 -1.9683>}]
 dyn_tunic_back_mid_01 [{shapeType=Box, extentX=12.125, extentY=3.125, extentZ=12.5, center=<1.2816 0 0>}]
 dyn_tunic_back_mid_02 [{shapeType=Box, extentX=12.125, extentY=3.125, extentZ=12.5, center=<1.2816 0 0>}]
 lowerarm_l [{shapeType=Capsule, length=24.8765, radius=4.5972, center=<13.8412 -0.0002 0>, rotation=<rot 90 0 0>}]
 pelvis [
  {shapeType=Capsule, length=8.1875, radius=13.4375, center=<-4.2317 -1.547 -0.0002>}
  {shapeType=Capsule, length=13.4375, radius=7.125, center=<-4.0876 -3.2313 -9.9699>, rotation=<rot 76.3572 180 180>}
 ]
 spine_01 [{shapeType=Capsule, length=15.25, radius=9.25}]
 spine_02 [{shapeType=Capsule, length=13.4375, radius=7.8125}]
 spine_03 [{shapeType=Capsule, length=14.375, radius=7.6875}]
 spine_04 [{shapeType=Capsule, length=10.25, radius=11}]
 spine_05 [{shapeType=Capsule, length=24.25, radius=8.25}]
 thigh_l [
  {shapeType=Capsule, length=26, radius=8, center=<-23.6137 -0.0002 0>, rotation=<rot 90 0 0>}
  {shapeType=Capsule, length=26, radius=8, center=<-23.6137 14.3588 0.0002>, rotation=<rot 90 0 0>}
 ]
 thigh_r [{shapeType=Capsule, length=33.5, radius=7, center=<22.8395 -0.0001 0>, rotation=<rot 90 0 0>}]
 upperarm_l [{shapeType=Capsule, length=19.8472, radius=4.4899, center=<13.0543 -0.3305 0>, rotation=<rot 90 0 0>}]
}
constraints [
 {bone1Name=dyn_sash_02, bone2Name=dyn_sash_01, swing1LimitDegrees=20, swing2LimitDegrees=15, twistLimitDegrees=5}
 {bone1Name=dyn_sash_03, bone2Name=dyn_sash_02, swing1LimitDegrees=25, swing2LimitDegrees=15, twistLimitDegrees=5}
 {bone1Name=dyn_sash_04, bone2Name=dyn_sash_03, swing1LimitDegrees=30, swing2LimitDegrees=15, twistLimitDegrees=5}
 {bone1Name=dyn_sash_05, bone2Name=dyn_sash_04, swing1LimitDegrees=35, swing2LimitDegrees=15, twistLimitDegrees=5}
 {bone1Name=dyn_sash_06, bone2Name=dyn_sash_05, swing1LimitDegrees=35, swing2LimitDegrees=15, twistLimitDegrees=5}
 {bone1Name=dyn_sash_07, bone2Name=dyn_sash_06, swing1LimitDegrees=35, swing2LimitDegrees=15, twistLimitDegrees=5}
 {bone1Name=dyn_sash_08, bone2Name=dyn_sash_07, swing1LimitDegrees=35, swing2LimitDegrees=15, twistLimitDegrees=5}
 {bone1Name=dyn_sash_01, bone2Name=dyn_shash_mstr, swing1LimitDegrees=10, swing2LimitDegrees=10, twistLimitDegrees=5}
 {
  bone1Name=spine_05
  bone2Name=spine_04
  swing1LimitDegrees=45
  swing1Motion=Free
  swing2LimitDegrees=45
  swing2Motion=Free
  twistLimitDegrees=45
  twistMotion=Free
 }
 {
  bone1Name=spine_02
  bone2Name=spine_01
  swing1LimitDegrees=45
  swing1Motion=Free
  swing2LimitDegrees=45
  swing2Motion=Free
  twistLimitDegrees=45
  twistMotion=Free
 }
 {
  bone1Name=spine_03
  bone2Name=spine_02
  swing1LimitDegrees=45
  swing1Motion=Free
  swing2LimitDegrees=45
  swing2Motion=Free
  twistLimitDegrees=45
  twistMotion=Free
 }
 {
  bone1Name=spine_04
  bone2Name=spine_03
  swing1LimitDegrees=45
  swing1Motion=Free
  swing2LimitDegrees=45
  swing2Motion=Free
  twistLimitDegrees=45
  twistMotion=Free
 }
 {
  bone1Name=spine_01
  bone2Name=pelvis
  swing1LimitDegrees=45
  swing1Motion=Free
  swing2LimitDegrees=45
  swing2Motion=Free
  twistLimitDegrees=45
  twistMotion=Free
 }
 {
  bone1Name=thigh_l
  bone2Name=pelvis
  swing1LimitDegrees=45
  swing1Motion=Free
  swing2LimitDegrees=45
  swing2Motion=Free
  twistLimitDegrees=45
  twistMotion=Free
 }
 {
  bone1Name=thigh_r
  bone2Name=pelvis
  swing1LimitDegrees=45
  swing1Motion=Free
  swing2LimitDegrees=45
  swing2Motion=Free
  twistLimitDegrees=45
  twistMotion=Free
 }
 {
  bone1Name=dyn_tunic_back_mid_01
  bone2Name=pelvis
  swing1LimitDegrees=45
  swing1Motion=Free
  swing2LimitDegrees=45
  swing2Motion=Free
  twistLimitDegrees=45
  twistMotion=Free
 }
 {
  bone1Name=dyn_tunic_back_mid_02
  bone2Name=dyn_tunic_back_mid_01
  swing1LimitDegrees=45
  swing1Motion=Free
  swing2LimitDegrees=45
  swing2Motion=Free
  twistLimitDegrees=45
  twistMotion=Free
 }
 {
  bone1Name=dyn_tunic_back_l_01
  bone2Name=pelvis
  swing1LimitDegrees=45
  swing1Motion=Free
  swing2LimitDegrees=45
  swing2Motion=Free
  twistLimitDegrees=45
  twistMotion=Free
 }
 {
  bone1Name=dyn_tunic_back_l_02
  bone2Name=dyn_tunic_back_l_01
  swing1LimitDegrees=45
  swing1Motion=Free
  swing2LimitDegrees=45
  swing2Motion=Free
  twistLimitDegrees=45
  twistMotion=Free
 }
 {
  bone1Name=upperarm_l
  bone2Name=clavicle_l
  swing1LimitDegrees=45
  swing1Motion=Free
  swing2LimitDegrees=45
  swing2Motion=Free
  twistLimitDegrees=45
  twistMotion=Free
 }
 {bone1Name=dyn_shash_mstr, bone2Name=clavicle_l, swing1LimitDegrees=0, swing2LimitDegrees=5, twistLimitDegrees=0}
 {
  bone1Name=clavicle_l
  bone2Name=spine_05
  swing1LimitDegrees=45
  swing1Motion=Free
  swing2LimitDegrees=45
  swing2Motion=Free
  twistLimitDegrees=45
  twistMotion=Free
 }
 {
  bone1Name=lowerarm_l
  bone2Name=upperarm_l
  swing1LimitDegrees=45
  swing1Motion=Free
  swing2LimitDegrees=45
  swing2Motion=Free
  twistLimitDegrees=45
  twistMotion=Free
 }
 {
  bone1Name=dyn_pouch_02
  bone2Name=pelvis
  swing1LimitDegrees=45
  swing1Motion=Free
  swing2LimitDegrees=45
  swing2Motion=Free
  twistLimitDegrees=45
  twistMotion=Free
 }
]
