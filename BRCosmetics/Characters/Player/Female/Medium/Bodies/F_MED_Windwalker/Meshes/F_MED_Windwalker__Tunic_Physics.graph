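# Physics asset for the Windwalker female medium body cosmetic, defining secondary motion for its tunic flaps and a pouch.
# Capsule bodies cover the pelvis, both thighs and short dyn_ bone chains for back, front and side flaps plus the pouch, all rooted at the pelvis.
# Flap segments use limited swing with locked or tightly limited twist.
# Thighs, front MSTR roots and the pouch root have fully free angular motion relative to the pelvis.
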PhysicsAsset /BRCosmetics/Characters/Player/Female/Medium/Bodies/F_MED_Windwalker/Meshes/F_MED_Windwalker__Tunic_Physics
bodies {
 dyn_pouch_05 [{shapeType=Capsule, length=0.505, radius=0.505, shapeName=dyn_pouch_05_capsule, rotation=<rot 0 -90 0>}]
 dyn_pouch_06 [{shapeType=Capsule, length=3.1409, radius=2.2064, shapeName=dyn_pouch_06_capsule, center=<3.7595 -0.5448 -0.0001>, rotation=<rot 90 0 0>}]
 dyn_tunic_back_l_01 [{shapeType=Capsule, length=6.5164, radius=2, shapeName=dyn_tunic_back_l_01_capsule, center=<4.3044 0.0001 0>}]
 dyn_tunic_back_l_02 [{shapeType=Capsule, length=8.8081, radius=2, shapeName=dyn_tunic_back_l_02_capsule, center=<5.5713 0.0001 0>, rotation=<rot 0 0 -7.0002>}]
 dyn_tunic_back_mid_01 [{shapeType=Capsule, length=6.7064, radius=1.25, shapeName=dyn_tunic_back_mid_01_capsule, center=<5.9537 0.0001 0>}]
 dyn_tunic_back_mid_02 [{shapeType=Capsule, length=11.2353, radius=3.5, shapeName=dyn_tunic_back_mid_02_capsule, center=<5.0718 0 0>}]
 dyn_tunic_back_r_01 [{shapeType=Capsule, length=6.51, radius=2, shapeName=dyn_tunic_back_r_01_capsule, center=<2.704 0.0001 0>}]
 dyn_tunic_back_r_02 [{shapeType=Capsule, length=8.75, radius=2, shapeName=dyn_tunic_back_r_02_capsule, center=<5.7236 0.0001 0.0001>, rotation=<rot 0 0 -20.9997>}]
 dyn_tunic_front_l_01 [{shapeType=Capsule, length=3.3218, radius=1.3805, shapeName=dyn_tunic_front_l_01_capsule}]
 dyn_tunic_front_l_02 [{shapeType=Capsule, length=8.2583, radius=2, shapeName=dyn_tunic_front_l_02_capsule, center=<5.6521 0 -0.0001>, rotation=<rot 0 0 62>}]
 dyn_tunic_front_l_03 [{shapeType=Capsule, length=0.505, radius=0.505, shapeName=dyn_tunic_front_l_03_capsule}]
 dyn_tunic_front_l_04 [{shapeType=Capsule, length=11.755, radius=2, shapeName=dyn_tunic_front_l_04_capsule, center=<8.5239 0 0>, rotation=<rot 0 -90 0>}]
 dyn_tunic_front_l_MSTR [{shapeType=Capsule, length=0.505, radius=0.505, shapeName=dyn_tunic_front_l_MSTR_capsule}]
 dyn_tunic_front_r_01 [{shapeType=Capsule, length=0.505, radius=0.505, shapeName=dyn_tunic_front_r_01_capsule}]
 dyn_tunic_front_r_02 [{shapeType=Capsule, length=12.7024, radius=2, shapeName=dyn_tunic_front_r_02_capsule, center=<6.9364 -2.1618 0.6619>, rotation=<rot 0 0 -72>}]
 dyn_tunic_front_r_03 [{shapeType=Capsule, length=0.505, radius=0.505, shapeName=dyn_tunic_front_r_03_capsule}]
 dyn_tunic_front_r_04 [{shapeType=Capsule, length=12.255, radius=2, shapeName=dyn_tunic_front_r_04_capsule, center=<7.9404 0.0001 0>, rotation=<rot 0 -90 0>}]
 dyn_tunic_front_r_MSTR [{shapeType=Capsule, length=0.505, radius=0.505, shapeName=dyn_tunic_front_r_MSTR_capsule}]
 dyn_tunic_side_l_03 [{shapeType=Capsule, length=0.505, radius=0.505, shapeName=dyn_tunic_side_l_03_capsule}]
 dyn_tunic_side_l_04 [{shapeType=Capsule, length=5.653, radius=1.75, shapeName=dyn_tunic_side_l_04_capsule, center=<6.4358 -0.0002 -0.0001>, rotation=<rot 0 0 62>}]
 dyn_tunic_side_l_MSTR [{shapeType=Capsule, length=0.505, radius=0.505, shapeName=dyn_tunic_side_l_MSTR_capsule}]
 dyn_tunic_side_r_01 [{shapeType=Capsule, length=0.505, radius=0.505, shapeName=dyn_tunic_side_r_01_capsule}]
 dyn_tunic_side_r_02 [{shapeType=Capsule, length=9.3762, radius=2, shapeName=dyn_tunic_side_r_02_capsule, center=<6.6096 0.0002 -0.0002>, rotation=<rot 0 0 -62>}]
 dyn_tunic_side_r_MSTR [{shapeType=Capsule, length=0.505, radius=0.505, shapeName=dyn_tunic_side_r_MSTR_capsule}]
 pelvis [{shapeType=Capsule, length=18.098, radius=8.15, center=<-5.4872 -0.7932 1.6418>}]
 thigh_l [{shapeType=Capsule, length=32.1736, radius=8.709, center=<-20.7522 0 -0.0001>, rotation=<rot 90 0 0>}]
 thigh_r [{shapeType=Capsule, length=26.7157, radius=9.1227, center=<19.1178 0 0>, rotation=<rot 90 0 0>}]
}
constraints [
 {bone1Name=dyn_tunic_back_r_01, bone2Name=pelvis, swing1LimitDegrees=10, swing2LimitDegrees=5, twistLimitDegrees=45, twistMotion=Locked}
 {
  bone1Name=dyn_tunic_back_r_02
  bone2Name=dyn_tunic_back_r_01
  swing1LimitDegrees=10
  swing2LimitDegrees=5
  twistLimitDegrees=45
  twistMotion=Locked
 }
 {bone1Name=dyn_tunic_back_mid_01, bone2Name=pelvis, swing1LimitDegrees=30, swing2LimitDegrees=10, twistLimitDegrees=45, twistMotion=Locked}
 {
  bone1Name=dyn_tunic_back_mid_02
  bone2Name=dyn_tunic_back_mid_01
  swing1LimitDegrees=30
  swing2LimitDegrees=10
  twistLimitDegrees=45
  twistMotion=Locked
 }
 {bone1Name=dyn_tunic_back_l_01, bone2Name=pelvis, swing1LimitDegrees=30, swing2LimitDegrees=10, twistLimitDegrees=45, twistMotion=Locked}
 {
  bone1Name=dyn_tunic_back_l_02
  bone2Name=dyn_tunic_back_l_01
  swing1LimitDegrees=30
  swing2LimitDegrees=10
  twistLimitDegrees=45
  twistMotion=Locked
 }
 {
  bone1Name=thigh_l
  bone2Name=pelvis
  swing1LimitDegrees=45
  swing1Motion=Free
  swing2LimitDegrees=45
  swing2Motion=Free
  twistLimitDegrees=45
  twistMotion=Free
 }
 {
  bone1Name=thigh_r
  bone2Name=pelvis
  swing1LimitDegrees=45
  swing1Motion=Free
  swing2LimitDegrees=45
  swing2Motion=Free
  twistLimitDegrees=45
  twistMotion=Free
 }
 {bone1Name=dyn_tunic_front_r_03, bone2Name=pelvis, swing1LimitDegrees=40, swing2LimitDegrees=20, twistLimitDegrees=45, twistMotion=Locked}
 {
  bone1Name=dyn_tunic_front_r_04
  bone2Name=dyn_tunic_front_r_03
  swing1LimitDegrees=40
  swing2LimitDegrees=20
  twistLimitDegrees=45
  twistMotion=Locked
 }
 {bone1Name=dyn_tunic_front_l_03, bone2Name=pelvis, swing1LimitDegrees=40, swing2LimitDegrees=20, twistLimitDegrees=45, twistMotion=Locked}
 {
  bone1Name=dyn_tunic_front_l_04
  bone2Name=dyn_tunic_front_l_03
  swing1LimitDegrees=40
  swing2LimitDegrees=30
  twistLimitDegrees=45
  twistMotion=Locked
 }
 {bone1Name=dyn_tunic_front_l_01, bone2Name=dyn_tunic_front_l_MSTR, swing1LimitDegrees=15, swing2LimitDegrees=25, twistLimitDegrees=2}
 {bone1Name=dyn_tunic_front_l_02, bone2Name=dyn_tunic_front_l_01, swing1LimitDegrees=15, swing2LimitDegrees=25, twistLimitDegrees=2}
 {bone1Name=dyn_tunic_front_r_01, bone2Name=dyn_tunic_front_r_MSTR, swing1LimitDegrees=20, swing2LimitDegrees=25, twistLimitDegrees=2}
 {bone1Name=dyn_tunic_front_r_02, bone2Name=dyn_tunic_front_r_01, swing1LimitDegrees=20, swing2LimitDegrees=25, twistLimitDegrees=2}
 {
  bone1Name=dyn_tunic_front_r_MSTR
  bone2Name=pelvis
  swing1LimitDegrees=45
  swing1Motion=Free
  swing2LimitDegrees=45
  swing2Motion=Free
  twistLimitDegrees=45
  twistMotion=Free
 }
 {
  bone1Name=dyn_tunic_front_l_MSTR
  bone2Name=pelvis
  swing1LimitDegrees=45
  swing1Motion=Free
  swing2LimitDegrees=45
  swing2Motion=Free
  twistLimitDegrees=45
  twistMotion=Free
 }
 {bone1Name=dyn_tunic_side_l_04, bone2Name=dyn_tunic_side_l_03, swing1LimitDegrees=5, swing2LimitDegrees=25, twistLimitDegrees=2}
 {bone1Name=dyn_tunic_side_l_03, bone2Name=dyn_tunic_side_l_MSTR, swing1LimitDegrees=15, swing2LimitDegrees=25, twistLimitDegrees=2}
 {bone1Name=dyn_tunic_side_l_MSTR, bone2Name=pelvis, swing1LimitDegrees=20, swing2LimitDegrees=25, twistLimitDegrees=2}
 {bone1Name=dyn_tunic_side_r_02, bone2Name=dyn_tunic_side_r_01, swing1LimitDegrees=10, swing2LimitDegrees=25, twistLimitDegrees=2}
 {bone1Name=dyn_tunic_side_r_01, bone2Name=dyn_tunic_side_r_MSTR, swing1LimitDegrees=20, swing2LimitDegrees=25, twistLimitDegrees=2}
 {bone1Name=dyn_tunic_side_r_MSTR, bone2Name=pelvis, swing1LimitDegrees=20, swing2LimitDegrees=25, twistLimitDegrees=2}
 {
  bone1Name=dyn_pouch_05
  bone2Name=pelvis
  swing1LimitDegrees=45
  swing1Motion=Free
  swing2LimitDegrees=45
  swing2Motion=Free
  twistLimitDegrees=45
  twistMotion=Free
 }
 {bone1Name=dyn_pouch_06, bone2Name=dyn_pouch_05, swing1LimitDegrees=30, swing2LimitDegrees=5, twistLimitDegrees=1}
]
