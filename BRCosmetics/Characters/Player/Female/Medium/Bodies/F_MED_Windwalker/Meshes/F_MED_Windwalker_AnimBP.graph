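# Post-process style AnimBlueprint for the F_MED_Windwalker cosmetic body, adding secondary motion to tunic, belt, strap and sash.
# It copies the parent mesh's pose, applies thigh-driven constraints, leg-up correctives and tunic physics, then caches BasePose.
# Belt RigidBody variants for front end, skydiving and default build Belt Pose, which feeds sash physics selected by backbling and crouch, with Control Rig corrections.
# Several alternative chains remain disconnected from the output.
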AnimBlueprint /BRCosmetics/Characters/Player/Female/Medium/Bodies/F_MED_Windwalker/Meshes/F_MED_Windwalker_AnimBP
target_skeleton=&/F_MED_Windwalker_Skeleton
graphs [
 AnimGraph {nodes [
   n1 = Control Rig(Source=Use cached pose 'Belt Pose'()) {
    alphaBoolBlend {blendOption=Linear}
    controlRigAssetReference {blueprintRigClass=&/F_MED_Windwalker_Crouch_CR.F_MED_Windwalker_Crouch_CR_C}
   }
   n2 = "Transform (Modify) Bone - Bone: dyn_pouch_05"(ComponentPose=Local To Component(LocalPose=@n1), Alpha=1, Translation="0,0,-1.5", Rotation="-30,0,-20", Scale="1,1,1") {
    rotationMode=BMM_Additive
    rotationSpace=BCS_BoneSpace
    translationMode=BMM_Additive
    translationSpace=BCS_BoneSpace
    alphaBoolBlend {blendOption=Linear}
    boneToModify {boneName=dyn_pouch_05}
    rotation {pitch=-30, roll=-20}
    translation {z=-1.5}
   }
   n3 = RigidBody!(ComponentPose=@n2, Alpha, OverrideWorldGravity:Vector=Get Gravity Override(), ExternalForce) -> (Pose) {
    bOverrideWorldGravity=true
    overridePhysicsAsset=&/F_MED_Windwalker_Belt_Physics
    simulationSpace=BaseBoneSpace
    alphaBoolBlend {blendOption=Linear}
    baseBoneRef {boneName=dyn_root_simspace}
    componentAppliedLinearAccClamp=<5000 5000 5000>
    componentLinearAccScale=<0.2 0.2 0.5>
    componentLinearVelScale=<0.2 0.2 0.2>
    overrideWorldGravity {z=-2500}
   }
   Blend Poses by bool(BlendPose_0=Component To Local(ComponentPose=@n3), BlendTime_0=0.2, BlendTime_1=0.2, bActiveValue=Get bIsCrouching()) {blendType=Linear}
   n4 = "Constraint - dyn_tunic_front_r_01 (R)"(ComponentPose=Local To Component(LocalPose=Use cached pose 'BasePose'()), Alpha=1, ConstraintWeights_0=1) {
    alphaBoolBlend {blendOption=Linear}
    boneToModify {boneName=dyn_tunic_front_r_01}
    constraintSetup [{transformType=Rotation, targetBone {boneName=thigh_r}}]
    constraintWeights [1]
   }
   n5 = "Constraint - dyn_tunic_front_l_01 (R)"(ComponentPose=@n4, Alpha=1, ConstraintWeights_0=1) {
    alphaBoolBlend {blendOption=Linear}
    boneToModify {boneName=dyn_tunic_front_l_01}
    constraintSetup [{transformType=Rotation, targetBone {boneName=thigh_l}}]
    constraintWeights [1]
   }
   n6 = "Constraint - dyn_tunic_front_r_03 (PP)"!(ComponentPose=@n5, Alpha=1, ConstraintWeights_0=0.7, ConstraintWeights_1:float=0.3) -> (Pose) {
    alphaBoolBlend {blendOption=Linear}
    boneToModify {boneName=dyn_tunic_front_r_03}
    constraintSetup [{targetBone {boneName=thigh_r}}, {targetBone {boneName=thigh_l}}]
    constraintWeights [1,1]
   }
   n7 = "Constraint - dyn_tunic_front_l_03 (PP)"!(ComponentPose=@n6, Alpha=1, ConstraintWeights_0=0.3, ConstraintWeights_1:float=0.7) -> (Pose) {
    alphaBoolBlend {blendOption=Linear}
    boneToModify {boneName=dyn_tunic_front_l_03}
    constraintSetup [{targetBone {boneName=thigh_r}}, {targetBone {boneName=thigh_l}}]
    constraintWeights [1,1]
   }
   n8 = "Transform (Modify) Bone - Bone: dyn_tunic_front_l_01"(ComponentPose=@n7, Alpha=1, Translation="0,6,0", Rotation="0,7,0", Scale="1,1,1") {
    rotationMode=BMM_Additive
    rotationSpace=BCS_BoneSpace
    scaleMode=BMM_Additive
    scaleSpace=BCS_BoneSpace
    translationMode=BMM_Additive
    translationSpace=BCS_BoneSpace
    alphaBoolBlend {blendOption=Linear}
    alphaScaleBiasClamp {inRange {max=0, min=-100}, outRange {max=0, min=1}}
    boneToModify {boneName=dyn_tunic_front_l_01}
    rotation {yaw=7}
    translation {y=6}
   }
   n9 = "Transform (Modify) Bone - Bone: dyn_tunic_front_l_02"(ComponentPose=@n8, Alpha=1, Translation="0,0,0", Rotation="0,0,0", Scale="1,1,1") {
    rotationMode=BMM_Additive
    rotationSpace=BCS_BoneSpace
    scaleMode=BMM_Additive
    scaleSpace=BCS_BoneSpace
    translationMode=BMM_Additive
    translationSpace=BCS_BoneSpace
    alphaBoolBlend {blendOption=Linear}
    alphaScaleBiasClamp {inRange {max=0, min=-100}, outRange {max=0, min=1}}
    boneToModify {boneName=dyn_tunic_front_l_02}
   }
   n10 = "Transform (Modify) Bone - Bone: dyn_tunic_front_r_01"(ComponentPose=@n9, Alpha=1, Translation="0,6,0", Rotation="0,20,0", Scale="1,1,1") {
    rotationMode=BMM_Additive
    rotationSpace=BCS_BoneSpace
    scaleMode=BMM_Additive
    scaleSpace=BCS_BoneSpace
    translationMode=BMM_Additive
    translationSpace=BCS_BoneSpace
    alphaBoolBlend {blendOption=Linear}
    alphaScaleBiasClamp {inRange {max=0, min=-100}, outRange {max=0, min=1}}
    boneToModify {boneName=dyn_tunic_front_r_01}
    rotation {yaw=20}
    translation {y=6}
   }
   n11 = "Constraint - dyn_tunic_front_r_01 (R)"(ComponentPose=Local To Component(LocalPose=Use cached pose 'BasePose'()), Alpha=1, ConstraintWeights_0=1) {
    alphaBoolBlend {blendOption=Linear}
    boneToModify {boneName=dyn_tunic_front_r_01}
    constraintSetup [{transformType=Rotation, targetBone {boneName=thigh_r}}]
    constraintWeights [1]
   }
   n12 = "Constraint - dyn_tunic_front_l_01 (R)"(ComponentPose=@n11, Alpha=1, ConstraintWeights_0=1) {
    alphaBoolBlend {blendOption=Linear}
    boneToModify {boneName=dyn_tunic_front_l_01}
    constraintSetup [{transformType=Rotation, targetBone {boneName=thigh_l}}]
    constraintWeights [1]
   }
   n13 = "Constraint - dyn_tunic_front_r_03 (PP)"(ComponentPose=@n12, Alpha=1, ConstraintWeights_0=0.7, ConstraintWeights_1=0.3) {
    alphaBoolBlend {blendOption=Linear}
    boneToModify {boneName=dyn_tunic_front_r_03}
    constraintSetup [{targetBone {boneName=thigh_r}}, {targetBone {boneName=thigh_l}}]
    constraintWeights [1,1]
   }
   n14 = "Constraint - dyn_tunic_front_l_03 (PP)"(ComponentPose=@n13, Alpha=1, ConstraintWeights_0=0.3, ConstraintWeights_1=0.7) {
    alphaBoolBlend {blendOption=Linear}
    boneToModify {boneName=dyn_tunic_front_l_03}
    constraintSetup [{targetBone {boneName=thigh_r}}, {targetBone {boneName=thigh_l}}]
    constraintWeights [1,1]
   }
   n15 = "Transform (Modify) Bone - Bone: dyn_tunic_front_l_01"(ComponentPose=@n14, Alpha=1, Translation="0,0,-12", Rotation="0,7,70", Scale="1,1,1") {
    rotationMode=BMM_Additive
    rotationSpace=BCS_BoneSpace
    scaleMode=BMM_Additive
    scaleSpace=BCS_BoneSpace
    translationMode=BMM_Additive
    translationSpace=BCS_BoneSpace
    alphaBoolBlend {blendOption=Linear}
    alphaScaleBiasClamp {inRange {max=0, min=-100}, outRange {max=0, min=1}}
    boneToModify {boneName=dyn_tunic_front_l_01}
    rotation {roll=70, yaw=7}
    translation {z=-12}
   }
   n16 = "Transform (Modify) Bone - Bone: dyn_tunic_front_l_02"(ComponentPose=@n15, Alpha=1, Translation="0,0,0", Rotation="0,-35,0", Scale="1,1,1") {
    rotationMode=BMM_Additive
    rotationSpace=BCS_BoneSpace
    scaleMode=BMM_Additive
    scaleSpace=BCS_BoneSpace
    translationMode=BMM_Additive
    translationSpace=BCS_BoneSpace
    alphaBoolBlend {blendOption=Linear}
    alphaScaleBiasClamp {inRange {max=0, min=-100}, outRange {max=0, min=1}}
    boneToModify {boneName=dyn_tunic_front_l_02}
    rotation {yaw=-35}
   }
   Blend Poses by bool(
    BlendPose_0=Component To Local(ComponentPose=@n16)
    BlendPose_1=Component To Local(ComponentPose=@n10)
    bActiveValue="float > float"(A=Get thighRotR() -> (thighRotR:double), B=-50) {is=title K2Node_CallFunction}
   ) {blendType=Linear}
   n17 = "Constraint - dyn_tunic_front_l_MSTR (P)"(ComponentPose=Local To Component(LocalPose=Copy Pose From Mesh() {bUseAttachedParent=true}), Alpha=1, ConstraintWeights_0=1) {
    alphaBoolBlend {blendOption=Linear}
    boneToModify {boneName=dyn_tunic_front_l_MSTR}
    constraintSetup [{targetBone {boneName=thigh_twist_01_l}}]
    constraintWeights [1]
   }
   n18 = "Constraint - dyn_tunic_front_r_MSTR (P)"(ComponentPose=@n17, Alpha=1, ConstraintWeights_0=1) {
    alphaBoolBlend {blendOption=Linear}
    boneToModify {boneName=dyn_tunic_front_r_MSTR}
    constraintSetup [{targetBone {boneName=thigh_r}}]
    constraintWeights [1]
   }
   n19 = "Constraint - dyn_tunic_side_r_MSTR (P)"(ComponentPose=@n18, Alpha=1, ConstraintWeights_0=1) {
    alphaBoolBlend {blendOption=Linear}
    boneToModify {boneName=dyn_tunic_side_r_MSTR}
    constraintSetup [{targetBone {boneName=thigh_r}}]
    constraintWeights [1]
   }
   n20 = "Constraint - dyn_tunic_side_l_MSTR (P)"(ComponentPose=@n19, Alpha=1, ConstraintWeights_0=1) {
    alphaBoolBlend {blendOption=Linear}
    boneToModify {boneName=dyn_tunic_side_l_MSTR}
    constraintSetup [{targetBone {boneName=thigh_twist_01_l}}]
    constraintWeights [1]
   }
   n21 = "Constraint - dyn_tunic_front_l_03 (P)"(ComponentPose=@n20, Alpha=0.5, ConstraintWeights_0=1) {
    alphaBoolBlend {blendOption=Linear}
    boneToModify {boneName=dyn_tunic_front_l_03}
    constraintSetup [{targetBone {boneName=thigh_l}}]
    constraintWeights [1]
   }
   n22 = "Constraint - dyn_tunic_front_r_03 (P)"(ComponentPose=@n21, Alpha=0.5, ConstraintWeights_0=1) {
    alphaBoolBlend {blendOption=Linear}
    boneToModify {boneName=dyn_tunic_front_r_03}
    constraintSetup [{targetBone {boneName=thigh_r}}]
    constraintWeights [1]
   }
   n23 = "Transform (Modify) Bone - Bone: dyn_tunic_front_r_MSTR"(ComponentPose=@n22, Alpha=1, Translation="0,0,0", Rotation="10,0,0", Scale="1,1,1") {
    rotationMode=BMM_Additive
    rotationSpace=BCS_BoneSpace
    scaleMode=BMM_Additive
    scaleSpace=BCS_BoneSpace
    translationMode=BMM_Additive
    translationSpace=BCS_BoneSpace
    alphaBoolBlend {blendOption=Linear}
    alphaScaleBiasClamp {inRange {max=0, min=-100}, outRange {max=0, min=1}}
    boneToModify {boneName=dyn_tunic_front_r_MSTR}
    rotation {pitch=10}
   }
   n24 = "Transform (Modify) Bone - Bone: dyn_tunic_front_l_MSTR"(ComponentPose=@n23, Alpha=1, Translation="0,0,1", Rotation="10,0,0", Scale="1,1,1") {
    rotationMode=BMM_Additive
    rotationSpace=BCS_BoneSpace
    scaleMode=BMM_Additive
    scaleSpace=BCS_BoneSpace
    translationMode=BMM_Additive
    translationSpace=BCS_BoneSpace
    alphaBoolBlend {blendOption=Linear}
    alphaScaleBiasClamp {inRange {max=0, min=-100}, outRange {max=0, min=1}}
    boneToModify {boneName=dyn_tunic_front_l_MSTR}
    rotation {pitch=10}
    translation {z=1}
   }
   n25 = "Constraint - dyn_pouch_05 (PPPP)"!(
    ComponentPose=@n24
    Alpha=1
    ConstraintWeights_0=0.0471
    ConstraintWeights_1:float=0.0785
    ConstraintWeights_2:float=0.015
    ConstraintWeights_3:float=0.8595
   ) -> (Pose) {
    alphaBoolBlend {blendOption=Linear}
    boneToModify {boneName=dyn_pouch_05}
    constraintSetup [{targetBone {boneName=spine_01}}, {targetBone {boneName=spine_02}}, {targetBone {boneName=spine_04}}, {targetBone {boneName=pelvis}}]
    constraintWeights [1,1,1,1]
   }
   n26 = "Constraint - dyn_strap (PPT)"!(ComponentPose=@n25, Alpha=1, ConstraintWeights_0=1, ConstraintWeights_1:float=0, ConstraintWeights_2:float=0) -> (Pose) {
    alphaBoolBlend {blendOption=Linear}
    boneToModify {boneName=dyn_strap}
    constraintSetup [{targetBone {boneName=clavicle_r}}, {targetBone {boneName=spine_05}}, {transformType=Translation, targetBone {boneName=upperarm_r}}]
    constraintWeights [1,1,1]
   }
   n27 = Get Leg Up R() -> (Leg Up R:double)
   n28 = "Transform (Modify) Bone - Bone: dyn_tunic_front_r_MSTR"(ComponentPose=@n26, Alpha=@n27, Translation="8,-8,7", Rotation="0,0,0", Scale="1,1,1") {
    rotationMode=BMM_Additive
    rotationSpace=BCS_BoneSpace
    scaleMode=BMM_Additive
    scaleSpace=BCS_BoneSpace
    translationMode=BMM_Additive
    translationSpace=BCS_BoneSpace
    alphaBoolBlend {blendOption=Linear}
    alphaScaleBiasClamp {inRange {max=0, min=-100}, outRange {max=0, min=1}}
    boneToModify {boneName=dyn_tunic_front_r_MSTR}
    translation=<8 -8 7>
   }
   n29 = "Transform (Modify) Bone - Bone: dyn_tunic_front_r_03"(ComponentPose=@n28, Alpha=@n27, Translation="0,0,0", Rotation="20,90,0", Scale="1,1,1") {
    rotationMode=BMM_Additive
    rotationSpace=BCS_BoneSpace
    scaleMode=BMM_Additive
    scaleSpace=BCS_BoneSpace
    translationMode=BMM_Additive
    translationSpace=BCS_BoneSpace
    alphaBoolBlend {blendOption=Linear}
    alphaScaleBiasClamp {inRange {max=0, min=-100}, outRange {max=0, min=1}}
    boneToModify {boneName=dyn_tunic_front_r_03}
    rotation {pitch=20, yaw=90}
   }
   n30 = "Transform (Modify) Bone - Bone: dyn_tunic_side_r_MSTR"(ComponentPose=@n29, Alpha=@n27, Translation="-3,4,4.5", Rotation="0,0,0", Scale="1,1,1") {
    rotationMode=BMM_Additive
    rotationSpace=BCS_BoneSpace
    scaleMode=BMM_Additive
    scaleSpace=BCS_BoneSpace
    translationMode=BMM_Additive
    translationSpace=BCS_BoneSpace
    alphaBoolBlend {blendOption=Linear}
    alphaScaleBiasClamp {inRange {max=0, min=-100}, outRange {max=0, min=1}}
    boneToModify {boneName=dyn_tunic_side_r_MSTR}
    translation=<-3 4 4.5>
   }
   n31 = Get Leg Up L() -> (Leg Up L:double)
   n32 = "Transform (Modify) Bone - Bone: dyn_tunic_front_l_MSTR"(ComponentPose=@n30, Alpha=@n31, Translation="8,8,4", Rotation="0,0,0", Scale="1,1,1") {
    rotationMode=BMM_Additive
    rotationSpace=BCS_BoneSpace
    scaleMode=BMM_Additive
    scaleSpace=BCS_BoneSpace
    translationMode=BMM_Additive
    translationSpace=BCS_BoneSpace
    alphaBoolBlend {blendOption=Linear}
    alphaScaleBiasClamp {inRange {max=0, min=-100}, outRange {max=0, min=1}}
    boneToModify {boneName=dyn_tunic_front_l_MSTR}
    translation=<8 8 4>
   }
   n33 = "Transform (Modify) Bone - Bone: dyn_tunic_front_l_03"(ComponentPose=@n32, Alpha=@n31, Translation="-8,0,0", Rotation="-20,60,0", Scale="1,1,1") {
    rotationMode=BMM_Additive
    rotationSpace=BCS_BoneSpace
    scaleMode=BMM_Additive
    scaleSpace=BCS_BoneSpace
    translationMode=BMM_Additive
    translationSpace=BCS_BoneSpace
    alphaBoolBlend {blendOption=Linear}
    alphaScaleBiasClamp {inRange {max=0, min=-100}, outRange {max=0, min=1}}
    boneToModify {boneName=dyn_tunic_front_l_03}
    rotation {pitch=-20, yaw=60}
    translation {x=-8}
   }
   n34 = "Transform (Modify) Bone - Bone: dyn_tunic_side_l_MSTR"(ComponentPose=@n33, Alpha=@n31, Translation="-3,3,2", Rotation="0,0,0", Scale="1,1,1") {
    rotationMode=BMM_Additive
    rotationSpace=BCS_BoneSpace
    scaleMode=BMM_Additive
    scaleSpace=BCS_BoneSpace
    translationMode=BMM_Additive
    translationSpace=BCS_BoneSpace
    alphaBoolBlend {blendOption=Linear}
    alphaScaleBiasClamp {inRange {max=0, min=-100}, outRange {max=0, min=1}}
    boneToModify {boneName=dyn_tunic_side_l_MSTR}
    translation=<-3 3 2>
   }
   n35 = Get Gravity Override()
   n36 = RigidBody(ComponentPose=@n34, OverrideWorldGravity=@n35) {
    overridePhysicsAsset=&/F_MED_Windwalker__Tunic_Physics
    simulationSpace=BaseBoneSpace
    alphaBoolBlend {blendOption=Linear}
    baseBoneRef {boneName=pelvis}
    componentLinearAccScale=<1 1 1>
    overrideWorldGravity {z=-3000}
   }
   n37 = RigidBody(ComponentPose=@n36, OverrideWorldGravity=@n35) {
    bOverrideWorldGravity=true
    overridePhysicsAsset=&/F_MED_Windwalker_Sash_Physics
    simulationSpace=BaseBoneSpace
    alphaBoolBlend {blendOption=Linear}
    baseBoneRef {boneName=spine_02}
    componentAppliedLinearAccClamp=<3000 3000 3000>
    componentLinearAccScale=<0.44 0.44 0.55>
    componentLinearVelScale=<0.2 0.2 0.2>
    overrideWorldGravity {z=-1000}
   }
   n38 = RigidBody(ComponentPose=@n37, OverrideWorldGravity=@n35) {
    bOverrideWorldGravity=true
    overridePhysicsAsset=&F_MED_Windwalker_Belt_Physics
    simulationSpace=BaseBoneSpace
    alphaBoolBlend {blendOption=Linear}
    baseBoneRef {boneName=pelvis}
    componentAppliedLinearAccClamp=<3000 3000 3000>
    componentLinearAccScale=<0.33 0.33 0.42>
    componentLinearVelScale=<0.2 0.2 0.2>
    overrideWorldGravity {z=-2500}
   }
   n39 = Control Rig(Source=Component To Local(ComponentPose=@n38)) {
    alphaBoolBlend {blendOption=Linear}
    controlRigAssetReference {blueprintRigClass=&/F_MED_Windwalker_CR.F_MED_Windwalker_CR_C}
   }
   Blend Poses by bool(BlendPose_0=@n39, BlendTime_0=0.5, BlendTime_1=0.5, bActiveValue=Get bIsTargeting()) {blendType=Linear}
   n40 = "Constraint - dyn_tunic_front_r_01 (R)"(ComponentPose=Local To Component(), Alpha=1, ConstraintWeights_0=1) {
    alphaBoolBlend {blendOption=Linear}
    boneToModify {boneName=dyn_tunic_front_r_01}
    constraintSetup [{transformType=Rotation, targetBone {boneName=thigh_r}}]
    constraintWeights [1]
   }
   n41 = "Constraint - dyn_tunic_front_l_01 (R)"(ComponentPose=@n40, Alpha=1, ConstraintWeights_0=1) {
    alphaBoolBlend {blendOption=Linear}
    boneToModify {boneName=dyn_tunic_front_l_01}
    constraintSetup [{transformType=Rotation, targetBone {boneName=thigh_l}}]
    constraintWeights [1]
   }
   n42 = "Constraint - dyn_tunic_front_r_03 (PP)"(ComponentPose=@n41, Alpha=1, ConstraintWeights_0=0.7, ConstraintWeights_1=0.3) {
    alphaBoolBlend {blendOption=Linear}
    boneToModify {boneName=dyn_tunic_front_r_03}
    constraintSetup [{targetBone {boneName=thigh_r}}, {targetBone {boneName=thigh_l}}]
    constraintWeights [1,1]
   }
   n43 = "Constraint - dyn_tunic_front_l_03 (PP)"(ComponentPose=@n42, Alpha=1, ConstraintWeights_0=0.3, ConstraintWeights_1=0.7) {
    alphaBoolBlend {blendOption=Linear}
    boneToModify {boneName=dyn_tunic_front_l_03}
    constraintSetup [{targetBone {boneName=thigh_r}}, {targetBone {boneName=thigh_l}}]
    constraintWeights [1,1]
   }
   Component To Local(ComponentPose=@n43)
   "Transform (Modify) Bone - Bone: dyn_tunic_front_r_MSTR"(Alpha=1, Translation="0,0,0", Rotation="0,0,0", Scale="1,1,1") {
    rotationMode=BMM_Additive
    rotationSpace=BCS_BoneSpace
    scaleMode=BMM_Additive
    scaleSpace=BCS_BoneSpace
    translationMode=BMM_Additive
    translationSpace=BCS_BoneSpace
    alphaBoolBlend {blendOption=Linear}
    alphaScaleBiasClamp {inRange {max=0, min=-100}, outRange {max=0, min=1}}
    boneToModify {boneName=dyn_tunic_front_r_MSTR}
   }
   n44 = Get Leg Up R()
   n45 = "Transform (Modify) Bone - Bone: dyn_tunic_front_r_MSTR"(Alpha=@n44, Translation="7,-6,11", Rotation="0,0,0", Scale="1,1,1") {
    rotationMode=BMM_Additive
    rotationSpace=BCS_BoneSpace
    scaleMode=BMM_Additive
    scaleSpace=BCS_BoneSpace
    translationMode=BMM_Additive
    translationSpace=BCS_BoneSpace
    alphaBoolBlend {blendOption=Linear}
    alphaScaleBiasClamp {inRange {max=0, min=-100}, outRange {max=0, min=1}}
    boneToModify {boneName=dyn_tunic_front_r_MSTR}
    translation=<7 -6 11>
   }
   n46 = "Transform (Modify) Bone - Bone: dyn_tunic_front_r_03"(ComponentPose=@n45, Alpha=@n44, Translation="0,0,0", Rotation="20,90,0", Scale="1,1,1") {
    rotationMode=BMM_Additive
    rotationSpace=BCS_BoneSpace
    scaleMode=BMM_Additive
    scaleSpace=BCS_BoneSpace
    translationMode=BMM_Additive
    translationSpace=BCS_BoneSpace
    alphaBoolBlend {blendOption=Linear}
    alphaScaleBiasClamp {inRange {max=0, min=-100}, outRange {max=0, min=1}}
    boneToModify {boneName=dyn_tunic_front_r_03}
    rotation {pitch=20, yaw=90}
   }
   n47 = Get Leg Up L()
   n48 = "Transform (Modify) Bone - Bone: dyn_tunic_front_l_MSTR"(ComponentPose=@n46, Alpha=@n47, Translation="8,8,2.5", Rotation="20,-10,0", Scale="1,1,1") {
    rotationMode=BMM_Additive
    rotationSpace=BCS_BoneSpace
    scaleMode=BMM_Additive
    scaleSpace=BCS_BoneSpace
    translationMode=BMM_Additive
    translationSpace=BCS_BoneSpace
    alphaBoolBlend {blendOption=Linear}
    alphaScaleBiasClamp {inRange {max=0, min=-100}, outRange {max=0, min=1}}
    boneToModify {boneName=dyn_tunic_front_l_MSTR}
    rotation {pitch=20, yaw=-10}
    translation=<8 8 2.5>
   }
   "Transform (Modify) Bone - Bone: dyn_tunic_front_l_03"(ComponentPose=@n48, Alpha=@n47, Translation="-8,0,0", Rotation="-20,60,0", Scale="1,1,1") {
    rotationMode=BMM_Additive
    rotationSpace=BCS_BoneSpace
    scaleMode=BMM_Additive
    scaleSpace=BCS_BoneSpace
    translationMode=BMM_Additive
    translationSpace=BCS_BoneSpace
    alphaBoolBlend {blendOption=Linear}
    alphaScaleBiasClamp {inRange {max=0, min=-100}, outRange {max=0, min=1}}
    boneToModify {boneName=dyn_tunic_front_l_03}
    rotation {pitch=-20, yaw=60}
    translation {x=-8}
   }
   "Transform (Modify) Bone - Bone: dyn_tunic_front_l_MSTR"(Alpha=1, Translation="0,0,2", Rotation="15,0,0", Scale="1,1,1") {
    rotationMode=BMM_Additive
    rotationSpace=BCS_BoneSpace
    scaleMode=BMM_Additive
    scaleSpace=BCS_BoneSpace
    translationMode=BMM_Additive
    translationSpace=BCS_BoneSpace
    alphaBoolBlend {blendOption=Linear}
    alphaScaleBiasClamp {inRange {max=0, min=-100}, outRange {max=0, min=1}}
    boneToModify {boneName=dyn_tunic_front_l_MSTR}
    rotation {pitch=15}
    translation {z=2}
   }
   "Transform (Modify) Bone - Bone: dyn_tunic_front_l_MSTR"(Alpha=1, Translation="0,0,2", Rotation="15,0,0", Scale="1,1,1") {
    rotationMode=BMM_Additive
    rotationSpace=BCS_BoneSpace
    scaleMode=BMM_Additive
    scaleSpace=BCS_BoneSpace
    translationMode=BMM_Additive
    translationSpace=BCS_BoneSpace
    alphaBoolBlend {blendOption=Linear}
    alphaScaleBiasClamp {inRange {max=0, min=-100}, outRange {max=0, min=1}}
    boneToModify {boneName=dyn_tunic_front_l_MSTR}
    rotation {pitch=15}
    translation {z=2}
   }
   RigidBody!(ComponentPose, Alpha=Get Pistol_ADS_Var() -> (Pistol_ADS_Var:double), ExternalForce) -> (Pose) {
    bOverrideWorldGravity=true
    overridePhysicsAsset=&/F_MED_Windwalker_Sash_Physics_Backpack
    simulationSpace=BaseBoneSpace
    alphaBoolBlend {blendOption=Linear}
    baseBoneRef {boneName=pelvis}
    componentAppliedLinearAccClamp=<3000 3000 3000>
    componentLinearAccScale=<0.33 0.33 0.42>
    componentLinearVelScale=<0.2 0.2 0.2>
    overrideWorldGravity {z=-1000}
   }
   RigidBody(Alpha=Get Wearing_Backbling() -> (Wearing_Backbling:double)) {
    bOverrideWorldGravity=true
    overridePhysicsAsset=&F_MED_Windwalker_Sash_Physics_Backpack
    simulationSpace=BaseBoneSpace
    alphaBoolBlend {blendOption=Linear}
    baseBoneRef {boneName=pelvis}
    componentAppliedLinearAccClamp=<3000 3000 3000>
    componentLinearAccScale=<0.33 0.33 0.42>
    componentLinearVelScale=<0.2 0.2 0.2>
    overrideWorldGravity {z=-1000}
   }
   RigidBody!(ComponentPose, Alpha, OverrideWorldGravity:Vector="0,0,-2500", ExternalForce) -> (Pose) {
    bOverrideWorldGravity=true
    overridePhysicsAsset=&F_MED_Windwalker_Belt_Physics
    simulationSpace=BaseBoneSpace
    alphaBoolBlend {blendOption=Linear}
    baseBoneRef {boneName=dyn_root_simspace}
    componentAppliedLinearAccClamp=<5000 5000 5000>
    componentLinearAccScale=<0.2 0.2 0.5>
    componentLinearVelScale=<0.2 0.2 0.2>
    overrideWorldGravity {z=-2500}
   }
   n49 = Get Gravity Override()
   n50 = RigidBody(
    ComponentPose=Local To Component(LocalPose=Use cached pose 'Belt Pose'())
    Alpha=Get NoBackpack Alpha() -> (NoBackpack Alpha:double)
    OverrideWorldGravity=@n49
   ) {
    bClampLinearTranslationLimitToRefPose=true
    bOverrideWorldGravity=true
    overridePhysicsAsset=&F_MED_Windwalker_Sash_Physics
    simulationSpace=BaseBoneSpace
    alphaBoolBlend {blendOption=Linear}
    baseBoneRef {boneName=spine_02}
    componentAppliedLinearAccClamp=<3000 3000 3000>
    componentLinearAccScale=<0.4 0.4 0.5>
    componentLinearVelScale=<0.2 0.2 0.35>
    overrideWorldGravity {z=-1000}
   }
   n51 = RigidBody(ComponentPose=@n50, Alpha=Get NoBackpack Alpha SkyD() -> (NoBackpack Alpha SkyD:double), OverrideWorldGravity=@n49) {
    bClampLinearTranslationLimitToRefPose=true
    bOverrideWorldGravity=true
    overridePhysicsAsset=&F_MED_Windwalker_Sash_Physics
    simulationSpace=BaseBoneSpace
    alphaBoolBlend {blendOption=Linear}
    baseBoneRef {boneName=spine_05}
    componentAppliedLinearAccClamp=<3000 3000 3000>
    componentLinearAccScale=<0.2 0.2 0.3>
    componentLinearVelScale=<0.1 0.1 0.15>
    overrideWorldGravity {z=-1000}
   }
   n52 = "Transform (Modify) Bone - Bone: dyn_shash_mstr"!(ComponentPose=@n51, bAlphaBoolEnabled:bool=Get bIsSkyDiving(), Translation="0,0,0", Rotation="0,-10,0", Scale="1,1,1") -> (Pose) {
    alphaInputType=Bool
    rotationMode=BMM_Additive
    rotationSpace=BCS_BoneSpace
    alphaBoolBlend {blendInTime=10, blendOption=Linear, blendOutTime=10}
    boneToModify {boneName=dyn_shash_mstr}
    rotation {yaw=-10}
   }
   n53 = Component To Local(ComponentPose=@n52)
   n54 = "Transform (Modify) Bone - Bone: dyn_pouch_05"(
    ComponentPose=Local To Component(LocalPose=Use cached pose 'Belt Pose'())
    Alpha=1
    Translation="0,0,-1.5"
    Rotation="-30,0,-20"
    Scale="1,1,1"
   ) {
    rotationMode=BMM_Additive
    rotationSpace=BCS_BoneSpace
    translationMode=BMM_Additive
    translationSpace=BCS_BoneSpace
    alphaBoolBlend {blendOption=Linear}
    boneToModify {boneName=dyn_pouch_05}
    rotation {pitch=-30, roll=-20}
    translation {z=-1.5}
   }
   n55 = RigidBody(ComponentPose=@n54, OverrideWorldGravity=@n49) {
    bClampLinearTranslationLimitToRefPose=true
    bOverrideWorldGravity=true
    overridePhysicsAsset=&F_MED_Windwalker_Sash_Physics
    simulationSpace=BaseBoneSpace
    alphaBoolBlend {blendOption=Linear}
    baseBoneRef {boneName=spine_02}
    componentAppliedLinearAccClamp=<5000 5000 5000>
    componentLinearAccScale=<0.3 0.3 0.4>
    componentLinearVelScale=<0.2 0.2 0.2>
    overrideWorldGravity {z=-1000}
   }
   n56 = Control Rig(Source=Component To Local(ComponentPose=@n55)) {
    alphaBoolBlend {blendOption=Linear}
    controlRigAssetReference {blueprintRigClass=&F_MED_Windwalker_Crouch_CR_C}
   }
   n57 = Get bIsCrouching()
   n58 = "Transform (Modify) Bone - Bone: dyn_sash_02"(
    ComponentPose=Local To Component(LocalPose=Use cached pose 'Belt Pose'())
    Alpha=Get L Clav Alpha() -> (L Clav Alpha:double)
    Translation="0,2,0"
    Rotation="-45,10,-20"
    Scale="1,1,1"
   ) {
    rotationMode=BMM_Additive
    rotationSpace=BCS_BoneSpace
    translationMode=BMM_Additive
    translationSpace=BCS_BoneSpace
    alphaBoolBlend {blendOption=Linear}
    boneToModify {boneName=dyn_sash_02}
    rotation=<rot -45 10 -20>
    translation {y=2}
   }
   n59 = RigidBody(ComponentPose=@n58, OverrideWorldGravity=Get Gravity Override()) {
    bClampLinearTranslationLimitToRefPose=true
    bOverrideWorldGravity=true
    overridePhysicsAsset=&F_MED_Windwalker_Sash_Physics_Backpack
    simulationSpace=BaseBoneSpace
    alphaBoolBlend {blendOption=Linear}
    baseBoneRef {boneName=spine_02}
    componentAppliedLinearAccClamp=<3000 3000 3000>
    componentLinearAccScale=<0.44 0.44 0.55>
    componentLinearVelScale=<0.2 0.2 0.2>
    overrideWorldGravity {z=-1000}
   }
   Output Pose(Result=Blend Poses by bool(
     BlendPose_0=Component To Local(ComponentPose=@n59)
     BlendPose_1=Blend Poses by bool(BlendPose_0=@n56, BlendPose_1=@n53, bActiveValue=@n57) {blendType=Linear}
     bActiveValue=Get Is_Wearing_Backbling() -> (Is_Wearing_Backbling:bool)
    ) {blendType=Linear}) {layerGroup=None}
   n60 = "Transform (Modify) Bone - Bone: dyn_root_simspace"(
    ComponentPose=Local To Component(LocalPose=Copy Pose From Mesh() {bUseAttachedParent=true})
    Alpha=1
    Translation="0,0,0"
    Rotation=Get sim_space_rotation() -> (sim_space_rotation:Rotator)
    Scale="1,1,1"
   ) {
    rotationMode=BMM_Replace
    rotationSpace=BCS_WorldSpace
    translationSpace=BCS_BoneSpace
    alphaBoolBlend {blendOption=Linear}
    alphaScaleBiasClamp {bInterpResult=true, interpSpeedDecreasing=2, interpSpeedIncreasing=2}
    boneToModify {boneName=dyn_root_simspace}
   }
   n61 = "Constraint - dyn_tunic_front_l_MSTR (P)"(ComponentPose=@n60, Alpha=1, ConstraintWeights_0=1) {
    alphaBoolBlend {blendOption=Linear}
    boneToModify {boneName=dyn_tunic_front_l_MSTR}
    constraintSetup [{targetBone {boneName=thigh_twist_01_l}}]
    constraintWeights [1]
   }
   n62 = "Constraint - dyn_tunic_front_r_MSTR (P)"(ComponentPose=@n61, Alpha=1, ConstraintWeights_0=1) {
    alphaBoolBlend {blendOption=Linear}
    boneToModify {boneName=dyn_tunic_front_r_MSTR}
    constraintSetup [{targetBone {boneName=thigh_twist_01_r}}]
    constraintWeights [1]
   }
   n63 = "Constraint - dyn_tunic_side_r_MSTR (P)"(ComponentPose=@n62, Alpha=1, ConstraintWeights_0=1) {
    alphaBoolBlend {blendOption=Linear}
    boneToModify {boneName=dyn_tunic_side_r_MSTR}
    constraintSetup [{targetBone {boneName=thigh_twist_01_r}}]
    constraintWeights [1]
   }
   n64 = "Constraint - dyn_tunic_side_l_MSTR (P)"(ComponentPose=@n63, Alpha=1, ConstraintWeights_0=1) {
    alphaBoolBlend {blendOption=Linear}
    boneToModify {boneName=dyn_tunic_side_l_MSTR}
    constraintSetup [{targetBone {boneName=thigh_twist_01_l}}]
    constraintWeights [1]
   }
   n65 = "Constraint - dyn_tunic_front_l_03 (P)"(ComponentPose=@n64, Alpha=0.5, ConstraintWeights_0=1) {
    alphaBoolBlend {blendOption=Linear}
    boneToModify {boneName=dyn_tunic_front_l_03}
    constraintSetup [{targetBone {boneName=thigh_l}}]
    constraintWeights [1]
   }
   n66 = "Constraint - dyn_tunic_front_r_03 (P)"(ComponentPose=@n65, Alpha=0.5, ConstraintWeights_0=1) {
    alphaBoolBlend {blendOption=Linear}
    boneToModify {boneName=dyn_tunic_front_r_03}
    constraintSetup [{targetBone {boneName=thigh_r}}]
    constraintWeights [1]
   }
   n67 = "Transform (Modify) Bone - Bone: dyn_tunic_front_r_MSTR"(ComponentPose=@n66, Alpha=1, Translation="0,0,0", Rotation="10,0,0", Scale="1,1,1") {
    rotationMode=BMM_Additive
    rotationSpace=BCS_BoneSpace
    scaleMode=BMM_Additive
    scaleSpace=BCS_BoneSpace
    translationMode=BMM_Additive
    translationSpace=BCS_BoneSpace
    alphaBoolBlend {blendOption=Linear}
    alphaScaleBiasClamp {inRange {max=0, min=-100}, outRange {max=0, min=1}}
    boneToModify {boneName=dyn_tunic_front_r_MSTR}
    rotation {pitch=10}
   }
   n68 = "Transform (Modify) Bone - Bone: dyn_tunic_front_l_MSTR"(ComponentPose=@n67, Alpha=1, Translation="0,0,1", Rotation="10,0,0", Scale="1,1,1") {
    rotationMode=BMM_Additive
    rotationSpace=BCS_BoneSpace
    scaleMode=BMM_Additive
    scaleSpace=BCS_BoneSpace
    translationMode=BMM_Additive
    translationSpace=BCS_BoneSpace
    alphaBoolBlend {blendOption=Linear}
    alphaScaleBiasClamp {inRange {max=0, min=-100}, outRange {max=0, min=1}}
    boneToModify {boneName=dyn_tunic_front_l_MSTR}
    rotation {pitch=10}
    translation {z=1}
   }
   n69 = "Constraint - dyn_pouch_05 (PPPP)"(ComponentPose=@n68, Alpha=1, ConstraintWeights_0=0.0471, ConstraintWeights_1=0.0785, ConstraintWeights_2=0.015, ConstraintWeights_3=0.8595) {
    alphaBoolBlend {blendOption=Linear}
    boneToModify {boneName=dyn_pouch_05}
    constraintSetup [{targetBone {boneName=spine_01}}, {targetBone {boneName=spine_02}}, {targetBone {boneName=spine_04}}, {targetBone {boneName=pelvis}}]
    constraintWeights [1,1,1,1]
   }
   n70 = "Constraint - dyn_strap (PPT)"(ComponentPose=@n69, Alpha=1, ConstraintWeights_0=1, ConstraintWeights_1=0, ConstraintWeights_2=0) {
    alphaBoolBlend {blendOption=Linear}
    boneToModify {boneName=dyn_strap}
    constraintSetup [{targetBone {boneName=clavicle_r}}, {targetBone {boneName=spine_05}}, {transformType=Translation, targetBone {boneName=upperarm_r}}]
    constraintWeights [1,1,1]
   }
   n71 = "Constraint - dyn_pouch_03 (PP)"!(ComponentPose=@n70, Alpha=1, ConstraintWeights_0=0.9828, ConstraintWeights_1:float=0.0172) -> (Pose) {
    alphaBoolBlend {blendOption=Linear}
    boneToModify {boneName=dyn_pouch_03}
    constraintSetup [{targetBone {boneName=pelvis}}, {targetBone {boneName=thigh_twist_01_r}}]
    constraintWeights [1,1]
   }
   n72 = "Constraint - dyn_pouch_01 (PP)"!(ComponentPose=@n71, Alpha=1, ConstraintWeights_0=0.9473, ConstraintWeights_1:float=0.0527) -> (Pose) {
    alphaBoolBlend {blendOption=Linear}
    boneToModify {boneName=dyn_pouch_01}
    constraintSetup [{targetBone {boneName=pelvis}}, {targetBone {boneName=thigh_twist_01_r}}]
    constraintWeights [1,1]
   }
   n73 = Get Leg Up R()
   n74 = "Transform (Modify) Bone - Bone: dyn_tunic_front_r_MSTR"(ComponentPose=@n72, Alpha=@n73, Translation="8,-8,7", Rotation="0,0,0", Scale="1,1,1") {
    rotationMode=BMM_Additive
    rotationSpace=BCS_BoneSpace
    scaleMode=BMM_Additive
    scaleSpace=BCS_BoneSpace
    translationMode=BMM_Additive
    translationSpace=BCS_BoneSpace
    alphaBoolBlend {blendOption=Linear}
    alphaScaleBiasClamp {inRange {max=0, min=-100}, outRange {max=0, min=1}}
    boneToModify {boneName=dyn_tunic_front_r_MSTR}
    translation=<8 -8 7>
   }
   n75 = "Transform (Modify) Bone - Bone: dyn_tunic_front_r_03"(ComponentPose=@n74, Alpha=@n73, Translation="0,0,0", Rotation="20,120,0", Scale="1,1,1") {
    rotationMode=BMM_Additive
    rotationSpace=BCS_BoneSpace
    scaleMode=BMM_Additive
    scaleSpace=BCS_BoneSpace
    translationMode=BMM_Additive
    translationSpace=BCS_BoneSpace
    alphaBoolBlend {blendOption=Linear}
    alphaScaleBiasClamp {inRange {max=0, min=-100}, outRange {max=0, min=1}}
    boneToModify {boneName=dyn_tunic_front_r_03}
    rotation {pitch=20, yaw=120}
   }
   n76 = "Transform (Modify) Bone - Bone: dyn_tunic_side_r_MSTR"(ComponentPose=@n75, Alpha=@n73, Translation="-3,3,4.5", Rotation="0,0,0", Scale="1,1,1") {
    rotationMode=BMM_Additive
    rotationSpace=BCS_BoneSpace
    scaleMode=BMM_Additive
    scaleSpace=BCS_BoneSpace
    translationMode=BMM_Additive
    translationSpace=BCS_BoneSpace
    alphaBoolBlend {blendOption=Linear}
    alphaScaleBiasClamp {inRange {max=0, min=-100}, outRange {max=0, min=1}}
    boneToModify {boneName=dyn_tunic_side_r_MSTR}
    translation=<-3 3 4.5>
   }
   n77 = Get Leg Up L()
   n78 = "Transform (Modify) Bone - Bone: dyn_tunic_front_l_MSTR"(ComponentPose=@n76, Alpha=@n77, Translation="8,10,4.5", Rotation="0,0,0", Scale="1,1,1") {
    rotationMode=BMM_Additive
    rotationSpace=BCS_BoneSpace
    scaleMode=BMM_Additive
    scaleSpace=BCS_BoneSpace
    translationMode=BMM_Additive
    translationSpace=BCS_BoneSpace
    alphaBoolBlend {blendOption=Linear}
    alphaScaleBiasClamp {inRange {max=0, min=-100}, outRange {max=0, min=1}}
    boneToModify {boneName=dyn_tunic_front_l_MSTR}
    translation=<8 10 4.5>
   }
   n79 = "Transform (Modify) Bone - Bone: dyn_tunic_front_l_03"(ComponentPose=@n78, Alpha=@n77, Translation="-8,0,0", Rotation="-20,60,0", Scale="1,1,1") {
    rotationMode=BMM_Additive
    rotationSpace=BCS_BoneSpace
    scaleMode=BMM_Additive
    scaleSpace=BCS_BoneSpace
    translationMode=BMM_Additive
    translationSpace=BCS_BoneSpace
    alphaBoolBlend {blendOption=Linear}
    alphaScaleBiasClamp {inRange {max=0, min=-100}, outRange {max=0, min=1}}
    boneToModify {boneName=dyn_tunic_front_l_03}
    rotation {pitch=-20, yaw=60}
    translation {x=-8}
   }
   n80 = "Transform (Modify) Bone - Bone: dyn_tunic_side_l_MSTR"(ComponentPose=@n79, Alpha=@n77, Translation="-3,3,2", Rotation="0,0,0", Scale="1,1,1") {
    rotationMode=BMM_Additive
    rotationSpace=BCS_BoneSpace
    scaleMode=BMM_Additive
    scaleSpace=BCS_BoneSpace
    translationMode=BMM_Additive
    translationSpace=BCS_BoneSpace
    alphaBoolBlend {blendOption=Linear}
    alphaScaleBiasClamp {inRange {max=0, min=-100}, outRange {max=0, min=1}}
    boneToModify {boneName=dyn_tunic_side_l_MSTR}
    translation=<-3 3 2>
   }
   n81 = RigidBody(ComponentPose=@n80, Alpha=Get Skydive Alpha Tunic() -> (Skydive Alpha Tunic:double), OverrideWorldGravity=Get Gravity Override()) {
    bClampLinearTranslationLimitToRefPose=true
    bOverrideWorldGravity=true
    overridePhysicsAsset=&F_MED_Windwalker__Tunic_Physics
    simulationSpace=BaseBoneSpace
    alphaBoolBlend {blendOption=Linear}
    baseBoneRef {boneName=pelvis}
    componentAppliedLinearAccClamp=<3000 3000 3000>
    componentLinearAccScale=<0.3 0.3 0.4>
    componentLinearVelScale=<0.2 0.2 0.2>
    overrideWorldGravity {z=-980}
   }
   Save cached pose 'BasePose'(Pose=Component To Local(ComponentPose=@n81)) {cacheName=BasePose}
   n82 = Control Rig(Source=Use cached pose 'BasePose'(), Alpha=Get Is Emoting() -> (Is Emoting:double)) {
    alphaBoolBlend {blendOption=Linear}
    controlRigAssetReference {blueprintRigClass=&F_MED_Windwalker_CR_C}
   }
   n83 = Get Gravity Override()
   n84 = RigidBody(ComponentPose=Local To Component(LocalPose=@n82), OverrideWorldGravity=@n83) {
    bClampLinearTranslationLimitToRefPose=true
    bOverrideWorldGravity=true
    overridePhysicsAsset=&F_MED_Windwalker_Belt_Physics
    simulationSpace=BaseBoneSpace
    alphaBoolBlend {blendOption=Linear}
    baseBoneRef {boneName=pelvis}
    componentAppliedLinearAccClamp=<3000 3000 3000>
    componentLinearAccScale=<0.4 0.4 0.5>
    componentLinearVelScale=<0.2 0.2 0.3>
    overrideWorldGravity {z=-2500}
   }
   n85 = Component To Local(ComponentPose=@n84)
   n86 = "Constraint - dyn_strap (PP)"!(
    ComponentPose=Local To Component(LocalPose=Use cached pose 'BasePose'())
    Alpha=Get Skydive Alpha() -> (Skydive Alpha:double)
    ConstraintWeights_0=0
    ConstraintWeights_1:float=1
   ) -> (Pose) {
    alphaBoolBlend {blendOption=Linear}
    boneToModify {boneName=dyn_strap}
    constraintSetup [{targetBone {boneName=upperarm_r}}, {targetBone {boneName=clavicle_r}}]
    constraintWeights [1,1]
   }
   n87 = "Transform (Modify) Bone - Bone: dyn_strap"(ComponentPose=@n86, Alpha="Get Y (Pitch)"() -> ("Y (Pitch)":double), Translation="-4,0,-2", Rotation="0,22.5,0.000", Scale="1,1,1") {
    rotationMode=BMM_Additive
    rotationSpace=BCS_BoneSpace
    scaleMode=BMM_Additive
    scaleSpace=BCS_BoneSpace
    translationMode=BMM_Additive
    translationSpace=BCS_BoneSpace
    alphaBoolBlend {blendOption=Linear}
    alphaScaleBiasClamp {inRange {max=0, min=-100}, outRange {max=0, min=1}}
    boneToModify {boneName=dyn_strap}
    rotation {yaw=22.5}
    translation {x=-4, z=-2}
   }
   n88 = "Transform (Modify) Bone - Bone: dyn_strap"(ComponentPose=@n87, Alpha="Get Z (Yaw)"() -> ("Z (Yaw)":double), Translation="-2.5,0,-6", Rotation="0,20,0.000", Scale="1,1,1") {
    rotationMode=BMM_Additive
    rotationSpace=BCS_BoneSpace
    scaleMode=BMM_Additive
    scaleSpace=BCS_BoneSpace
    translationMode=BMM_Additive
    translationSpace=BCS_BoneSpace
    alphaBoolBlend {blendOption=Linear}
    alphaScaleBiasClamp {inRange {max=0, min=-100}, outRange {max=0, min=1}}
    boneToModify {boneName=dyn_strap}
    rotation {yaw=20}
    translation {x=-2.5, z=-6}
   }
   n89 = RigidBody(ComponentPose=@n88, Alpha=0.6, OverrideWorldGravity=@n83) {
    bClampLinearTranslationLimitToRefPose=true
    bOverrideWorldGravity=true
    overridePhysicsAsset=&F_MED_Windwalker_Belt_Physics
    simulationSpace=BaseBoneSpace
    alphaBoolBlend {blendOption=Linear}
    baseBoneRef {boneName=pelvis}
    componentAppliedLinearAccClamp=<3000 3000 3000>
    componentLinearAccScale=<0.2 0.2 0.3>
    componentLinearVelScale=<0.2 0.2 0.2>
    overrideWorldGravity {z=-2500}
   }
   n90 = Component To Local(ComponentPose=@n89)
   n91 = Get bIsSkyDiving()
   n92 = RigidBody(ComponentPose=Local To Component(LocalPose=Use cached pose 'BasePose'()), OverrideWorldGravity=Get Gravity Override()) {
    bClampLinearTranslationLimitToRefPose=true
    bOverrideWorldGravity=true
    overridePhysicsAsset=&F_MED_Windwalker_Belt_Physics
    simulationSpace=BaseBoneSpace
    alphaBoolBlend {blendOption=Linear}
    baseBoneRef {boneName=pelvis}
    componentAppliedLinearAccClamp=<5000 5000 5000>
    componentLinearAccScale=<0.4 0.4 0.5>
    componentLinearVelScale=<0.3 0.3 0.3>
    overrideWorldGravity {z=-2500}
   }
   Save cached pose 'Belt Pose'(Pose=Blend Poses by bool(
     BlendPose_0=Component To Local(ComponentPose=@n92)
     BlendPose_1=Blend Poses by bool(BlendPose_0=@n90, BlendPose_1=@n85, BlendTime_0=0.35, bActiveValue=@n91) {blendType=Linear}
     bActiveValue=Get bIsInFrontEnd()
    ) {blendType=Linear}) {cacheName=Belt Pose}
   Get Not_Wearing_Backbling() -> (Not_Wearing_Backbling:double)
  ]}
 EventGraph {nodes [
   Event BlueprintUpdateAnimation()
   > n1 = Sequence!(execute) -> (then_0, then_1, then_2:Exec, then_3:Exec, then_4:Exec, then_5:Exec, then_6:Exec, then_7:Exec)
   n2 = Gravity Override State Function DEPRECATED(
    execute=@n1.then_5
    CharacterPartAnimInstance=Self-Reference()
    GameTimeInSeconds=GetGameTimeInSeconds()
    GravityOverrideParams=Get Gravity Override Params()
   )
   > n3 = Branch(Condition=Get bIsDBNOCarried())
   > Set Gravity Override(Gravity Override="0,0,-980")
   Set Gravity Override(execute=@n3.else, Gravity Override=@n2)
   n4 = GetDeltaTransformFromRefPose(self=GetOwnerSkeletalMeshComponent(), BoneName=thigh_r, BaseName=pelvis)
   Set thighRotR(execute=@n1.then_1, thighRotR:double=BreakRotator(InRot=BreakTransform(InTransform=@n4).Rotation).Yaw) -> (Output_Get:double)
   n5 = BreakTransform(InTransform=GetDeltaTransformFromRefPose(self=GetOwnerSkeletalMeshComponent(), BoneName=clavicle_l, BaseName=spine_05))
   > Set L Clav Alpha(L Clav Alpha:double=MapRangeClamped(Value=BreakRotator(InRot=@n5.Rotation).Pitch, InRangeA=0, InRangeB=50)) -> (Output_Get:double)
   n6 = Get Body Mesh() -> (Body Mesh:Skeletal Mesh Component Object Reference)
   Set Leg Up L(
    execute=@n1.then_3
    Leg Up L:double=GetAlphaFromReferenceTransformDelta(SkeletalMeshComponent=@n6, BoneName=thigh_l, BaseName=pelvis, DeltaComponentType=QuaternionTwist, InputRange=(X=15,Y=100), TwistAxis="0,0,1")
   ) -> (Output_Get:double)
   > Set Leg Up R(Leg Up R:double=GetAlphaFromReferenceTransformDelta(SkeletalMeshComponent=@n6, BoneName=thigh_r, BaseName=pelvis, DeltaComponentType=QuaternionTwist, InputRange=(X=-5,Y=70), TwistAxis="0,0,1")) -> (Output_Get:double)
   n7 = Cast To FortPlayerAnimInstance(execute=@n1.then_2, Object=GetAnimInstance(self=GetOwnerSkeletalMeshComponent()))
   > Set Pistol_ADS(Pistol_ADS:bool=false) -> (Output_Get:bool)
   > n8 = Switch on EFortWeaponCoreAnimation!(execute, Selection:EFortWeaponCoreAnimation Enum=Get WeaponCoreAnim(self=@n7)) -> (
    Melee:Exec
    Pistol:Exec
    Shotgun:Exec
    PaperBlueprint:Exec
    Rifle:Exec
    MeleeOneHand:Exec
    MachinePistol:Exec
    RocketLauncher:Exec
    GrenadeLauncher:Exec
    GoingCommando:Exec
    AssaultRifle:Exec
    TacticalShotgun:Exec
    SniperRifle:Exec
    TrapPlacement:Exec
    ShoulderLauncher:Exec
    AbilityDecoTool:Exec
    Crossbow:Exec
    C4:Exec
    RemoteControl:Exec
    DualWield:Exec
    AR_BullPup:Exec
    AR_ForwardGrip:Exec
    MedPackPaddles:Exec
    SMG_P90:Exec
    AR_DrumGun:Exec
    Consumable_Small:Exec
    Consumable_Large:Exec
    Balloon:Exec
    MountedTurret:Exec
    CreativeTool:Exec
    ExplosiveBow:Exec
    AshtonIndigo:Exec
    AshtonChicago:Exec
    MeleeDualWield:Exec
    Unarmed:Exec
   )
   Branch(execute=[@n8.Pistol, @n8.Rifle], Condition=Get bIsTargeting())
   > Set Pistol_ADS(Pistol_ADS=true)
   n9 = GetDeltaTransformFromRefPose(self=GetOwnerSkeletalMeshComponent(), BoneName=thigh_l, BaseName=pelvis)
   Set thighRotL(execute=@n1.then_0, thighRotL:double=BreakRotator(InRot=BreakTransform(InTransform=@n9).Rotation).Yaw) -> (Output_Get:double)
   n10 = GetOwnerSkeletalMeshComponent()
   n11 = BreakRotator(InRot=BreakTransform(InTransform=GetDeltaTransformFromRefPose(self=@n10, BoneName=upperarm_r, BaseName=spine_05)).Rotation)
   > "Set Y (Pitch)"("Y (Pitch)":double=MapRangeClamped(Value=@n11.Pitch, InRangeA=35, InRangeB=60)) -> (Output_Get:double)
   > "Set Z (Yaw)"("Z (Yaw)":double=MapRangeClamped(Value="Absolute (Float)"(A=@n11.Yaw), InRangeA=0, InRangeB=50)) -> (Output_Get:double)
   > n12 = Branch(Condition=Get bIsSkyDiving())
   > Set Skydive Alpha(Skydive Alpha:double=0.5) -> (Output_Get:double)
   Set Skydive Alpha(execute=@n12.else, Skydive Alpha=0)
   > Set Is Emoting(Is Emoting:double=SelectFloat(A=0.4, B=1, bPickA=Get bIsPlayingEmote())) -> (Output_Get:double)
   n13 = Branch(execute=@n1.then_4, Condition=Get Pistol_ADS() -> (Pistol_ADS:bool))
   n14 = Branch(execute=@n13.else, Condition=Get Is_Wearing_Backbling())
   > Set Wearing_Backbling(Wearing_Backbling:double=1) -> (Output_Get:double)
   > Set Not_Wearing_Backbling(Not_Wearing_Backbling:double=0) -> (Output_Get:double)
   > Set Pistol_ADS_Var(Pistol_ADS_Var:double=0) -> (Output_Get:double)
   Set Wearing_Backbling(execute=@n14.else, Wearing_Backbling=0)
   > Set Not_Wearing_Backbling(Not_Wearing_Backbling=1)
   > Set Pistol_ADS_Var(Pistol_ADS_Var=0)
   n15 = ComplexSimSpaceFunction DEPRECATED!(
    execute=@n1.then_6
    CharacterPartAnimInstance:Custom Character Part Anim Instance Object Reference=Self-Reference()
    InterpSpeedMax:double=9
    PawnSpeedMax:double=20
    RInterpErrorTolerance:double=1
    DeltaTime:double=0
    DebugDraw:bool=false
    sim_space_rotation_in:Rotator=Get sim_space_rotation()
    inComponentSpaceIn:bool=Get InComponentSpace() -> (InComponentSpace:bool)
    inComponentSpaceLastFrameIn:bool=Get InComponentSpaceLastFrame() -> (InComponentSpaceLastFrame:bool)
    lockedOnSimSpaceIn:bool=Get LockedOnSimSpace() -> (LockedOnSimSpace:bool)
    InterpSpeedIn:double=Get InterpSpeed() -> (InterpSpeed:double)
   ) -> (
    then
    sim_space_rotation:Rotator
    inComponentSpaceOut:bool
    inComponentSpaceLastFrameOut:bool
    lockedOnSimSpaceOut:bool
    InterpSpeedOut:double
   )
   > Set sim_space_rotation(sim_space_rotation:Rotator=@n15) -> (Output_Get:Rotator)
   > Set InComponentSpace(InComponentSpace:bool=@n15.inComponentSpaceOut) -> (Output_Get:bool)
   > Set InComponentSpaceLastFrame(InComponentSpaceLastFrame:bool=@n15.inComponentSpaceLastFrameOut) -> (Output_Get:bool)
   > Set LockedOnSimSpace(LockedOnSimSpace:bool=@n15.lockedOnSimSpaceOut) -> (Output_Get:bool)
   > Set InterpSpeed(InterpSpeed:double=@n15.InterpSpeedOut) -> (Output_Get:double)
   n16 = Branch(execute=@n1.then_7, Condition=Get bIsSkyDiving())
   > Set NoBackpack Alpha(NoBackpack Alpha:double=0) -> (Output_Get:double)
   > Set NoBackpack Alpha SkyD(NoBackpack Alpha SkyD:double=1) -> (Output_Get:double)
   > Set Skydive Alpha Tunic(Skydive Alpha Tunic:double=0.4) -> (Output_Get:double)
   Set Skydive Alpha Tunic(execute=@n16.else, Skydive Alpha Tunic=1)
   > Set NoBackpack Alpha SkyD(NoBackpack Alpha SkyD=0)
   > Set NoBackpack Alpha(NoBackpack Alpha=1)
   Event BlueprintInitializeAnimation()
   > Set Body Mesh(Body Mesh:Skeletal Mesh Component Object Reference=GetPartSkeletalMeshComponent()) -> (Output_Get:Skeletal Mesh Component Object Reference)
   > Set Anim Inst(Anim Inst:Anim Instance Object Reference=Get Anim Inst() -> (Anim Inst:Anim Instance Object Reference)) -> (Output_Get:Anim Instance Object Reference)
   Event OnCharacterCustomizationLoaded()
   n17 = CheckForBackbling()
   n18 = OR Boolean(A=@n17, B=@n17.bIsBackblingLong)
   > n19 = Set Is_Wearing_Backbling(Is_Wearing_Backbling:bool=@n18) -> (Output_Get:bool)
   ResetDynamics(execute=[@n19.then, @n20.then], InTeleportType=TeleportPhysics)
   > ResetDynamics(InTeleportType=ResetPhysics)
   n20 = Set Not_Wearing_Backbling(Not_Wearing_Backbling=SelectFloat(A=0, B=1, bPickA=@n18))
   GetOwnerSkeletalMeshComponent()
   GetAnimInstance()
  ]}
]
variables [
 Anim Inst {type=AnimInstance, default=null}
 Body Mesh {type=SkeletalMeshComponent}
 Gravity Override {type=Vector}
 Gravity Override Params {
  type=GravityOverrideParamsStruct
  default {
   onGround_Moving {gravityOverride=<0 0 -1000>}
   onGround_Standing {gravityOverride=<0 0 -1000>}
   skydive_DivingDown {gravityOverride=<0 0 0>}
   skydive_FloatingDown {gravityOverride=<0 0 0>}
   skydive_GoingUp {gravityOverride=<0 0 -100>}
   skydive_Parachuting {gravityOverride=<0 0 -980>}
  }
 }
 InComponentSpace {type=boolean}
 InComponentSpaceLastFrame {type=boolean}
 InterpSpeed {type=number}
 Is Emoting {type=number}
 Is_Wearing_Backbling {type=boolean}
 L Clav Alpha {type=number}
 Leg Up L {type=number}
 Leg Up R {type=number}
 LockedOnSimSpace {type=boolean}
 NoBackpack Alpha {type=number}
 NoBackpack Alpha SkyD {type=number}
 Not_Wearing_Backbling {type=number}
 Pistol_ADS {type=boolean}
 Pistol_ADS_Var {type=number}
 sim_space_rotation {type=Rotator}
 Skydive Alpha {type=number}
 Skydive Alpha Tunic {type=number}
 thighRotL {type=number}
 thighRotR {type=number}
 Wearing_Backbling {type=number}
 "Y (Pitch)" {type=number}
 "Z (Yaw)" {type=number}
]
