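# Physics asset for the ponytail part of the F_MED_Windwalker female character body, for simulating the hair's motion.
# Three tapering capsules form the ponytail chain from dyn_hair_mstr, each joint using identical limited swing and very tight twist.
# Capsules on the hair master, both neck bones and two upper spine bones give the head and upper-body volume.
# Constraints linking those bodies leave swing and twist free, so only the ponytail joints are angularly limited.
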PhysicsAsset /BRCosmetics/Characters/Player/Female/Medium/Bodies/F_MED_Windwalker/Meshes/Parts/F_MED_Windwalker_Ponytail_Physics
bodies {
 dyn_hair_mstr [{shapeType=Capsule, length=2.5805, radius=6.9822, center=<0 -3.4 -2.4333>}]
 dyn_ponytail_01 [{shapeType=Capsule, length=4.9207, radius=2.8307, center=<4.3249 0 -0.0001>, rotation=<rot 90 0 0>}]
 dyn_ponytail_02 [{shapeType=Capsule, length=3.6632, radius=2.4641, center=<4.3249 0 -0.0001>, rotation=<rot 90 0 0>}]
 dyn_ponytail_03 [{shapeType=Capsule, length=4.1933, radius=2.1902, center=<4.3249 0 -0.0001>, rotation=<rot 90 0 0>}]
 neck_01 [{shapeType=Capsule, length=0, radius=5.5, center=<0.8889 0.4923 0>, rotation=<rot 90 0 0>}]
 neck_02 [{shapeType=Capsule, length=0, radius=5.8125, rotation=<rot 90 0 0>}]
 spine_04 [{shapeType=Capsule, length=27.625, radius=9.25, center=<6.1244 -2.472 0.0001>, rotation=<rot 0 -100.0001 0>}]
 spine_05 [
  {shapeType=Capsule, length=27.625, radius=9.125, center=<5.8058 0.7722 0.0001>, rotation=<rot 0 -90.0001 0>}
  {shapeType=Capsule, length=24.875, radius=5.125, shapeName=spine_05_capsule, center=<13.256 2.4999 0>}
 ]
}
constraints [
 {bone1Name=dyn_ponytail_01, bone2Name=dyn_hair_mstr, swing1LimitDegrees=45, swing2LimitDegrees=20, twistLimitDegrees=2}
 {bone1Name=dyn_ponytail_02, bone2Name=dyn_ponytail_01, swing1LimitDegrees=45, swing2LimitDegrees=20, twistLimitDegrees=2}
 {bone1Name=dyn_ponytail_03, bone2Name=dyn_ponytail_02, swing1LimitDegrees=45, swing2LimitDegrees=20, twistLimitDegrees=2}
 {
  bone1Name=dyn_hair_mstr
  bone2Name=neck_02
  swing1LimitDegrees=45
  swing1Motion=Free
  swing2LimitDegrees=45
  swing2Motion=Free
  twistLimitDegrees=45
  twistMotion=Free
 }
 {
  bone1Name=neck_02
  bone2Name=neck_01
  swing1LimitDegrees=45
  swing1Motion=Free
  swing2LimitDegrees=45
  swing2Motion=Free
  twistLimitDegrees=45
  twistMotion=Free
 }
 {
  bone1Name=neck_01
  bone2Name=spine_05
  swing1LimitDegrees=45
  swing1Motion=Free
  swing2LimitDegrees=45
  swing2Motion=Free
  twistLimitDegrees=45
  twistMotion=Free
 }
 {
  bone1Name=spine_05
  bone2Name=spine_04
  swing1LimitDegrees=45
  swing1Motion=Free
  swing2LimitDegrees=45
  swing2Motion=Free
  twistLimitDegrees=45
  twistMotion=Free
 }
]
