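# Physics asset for the strands part of the female medium Windwalker player body, giving its dynamic hair bones capsule collision and joints.
# Front, left and right strand chains hang from dyn_hair_mstr, with the left and right chains branching at their first bone.
# Strand joints lock twist and carry per-joint swing limits, while the master joins spine_05, which has a larger capsule, with free swing and twist.
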PhysicsAsset /BRCosmetics/Characters/Player/Female/Medium/Bodies/F_MED_Windwalker/Meshes/Parts/F_MED_Windwalker_strands_Physics
bodies {
 dyn_hair_mstr [{shapeType=Capsule, length=3.5805, radius=4.7322, center=<0 -1.7664 -7.7324>, rotation=<rot 90 0 0>}]
 dyn_strand_front_01 [{shapeType=Capsule, length=0.3, radius=0.3, center=<0.7398 0 0>, rotation=<rot 90 0 0>}]
 dyn_strand_front_02 [{shapeType=Capsule, length=1.6, radius=0.5, center=<3 0 -0.0001>, rotation=<rot 90 0 0>}]
 dyn_strand_front_03 [{shapeType=Capsule, length=1.6, radius=1.25, center=<2.9398 -0.7167 0.0001>, rotation=<rot 90 0 0>}]
 dyn_strand_front_04 [{shapeType=Capsule, length=0.3, radius=0.3, center=<0.7398 0 0>, rotation=<rot 90 0 0>}]
 dyn_strand_front_05 [{shapeType=Capsule, length=0.3, radius=0.3, center=<0.7398 0 0>, rotation=<rot 90 0 0>}]
 dyn_strand_l_01 [{shapeType=Capsule, length=1.3828, radius=0.8601, center=<2.034 0.4322 0>, rotation=<rot 90 0 0>}]
 dyn_strand_l_02 [{shapeType=Capsule, length=0.505, radius=0.505, center=<-0.1513 -0.844 -1.7693>, rotation=<rot 25.8367 -90 15.6061>}]
 dyn_strand_l_03 [{shapeType=Capsule, length=1.767, radius=0.6302, center=<2.2292 0.0684 -0.3684>, rotation=<rot 0 -94.8336 -85.0663>}]
 dyn_strand_l_04 [{shapeType=Capsule, length=0.505, radius=0.505, center=<1.7809 -0.2432 -0.2239>, rotation=<rot 0 105.1444 -71.4793>}]
 dyn_strand_r_01 [{shapeType=Capsule, length=1.4248, radius=0.7934, center=<2.1797 -0.0047 -0.3467>, rotation=<rot 0 -20.4826 -83.0601>}]
 dyn_strand_r_02 [{shapeType=Capsule, length=1.1128, radius=0.365, center=<1.7706 0.0736 -0.3208>, rotation=<rot 0 -83.3896 -75.2177>}]
 dyn_strand_r_03 [{shapeType=Capsule, length=0.505, radius=0.755, center=<2.3301 -0.3913 0.0002>, rotation=<rot 0 -90 0>}]
 dyn_strand_r_04 [{shapeType=Capsule, length=2.6982, radius=0.9748, center=<2.9969 0.4018 -0.6783>, rotation=<rot 0 -91.3773 -52.0955>}]
 spine_05 [{shapeType=Capsule, length=16.75, radius=7.75, center=<8.6651 -1.2278 0>, rotation=<rot 0 -90 0>}]
}
constraints [
 {bone1Name=dyn_strand_l_01, bone2Name=dyn_hair_mstr, swing1LimitDegrees=45, swing2LimitDegrees=45, twistLimitDegrees=45, twistMotion=Locked}
 {bone1Name=dyn_strand_l_02, bone2Name=dyn_strand_l_01, swing1LimitDegrees=45, swing2LimitDegrees=45, twistLimitDegrees=45, twistMotion=Locked}
 {bone1Name=dyn_strand_l_03, bone2Name=dyn_strand_l_01, swing1LimitDegrees=45, swing2LimitDegrees=45, twistLimitDegrees=45, twistMotion=Locked}
 {bone1Name=dyn_strand_l_04, bone2Name=dyn_strand_l_03, swing1LimitDegrees=45, swing2LimitDegrees=45, twistLimitDegrees=45, twistMotion=Locked}
 {bone1Name=dyn_strand_r_01, bone2Name=dyn_hair_mstr, swing1LimitDegrees=80, swing2LimitDegrees=45, twistLimitDegrees=45, twistMotion=Locked}
 {bone1Name=dyn_strand_r_04, bone2Name=dyn_strand_r_01, swing1LimitDegrees=90, swing2LimitDegrees=45, twistLimitDegrees=45, twistMotion=Locked}
 {bone1Name=dyn_strand_r_03, bone2Name=dyn_strand_r_04, swing1LimitDegrees=45, swing2LimitDegrees=45, twistLimitDegrees=45, twistMotion=Locked}
 {bone1Name=dyn_strand_r_02, bone2Name=dyn_strand_r_01, swing1LimitDegrees=45, swing2LimitDegrees=45, twistLimitDegrees=45, twistMotion=Locked}
 {
  bone1Name=dyn_strand_front_01
  bone2Name=dyn_hair_mstr
  swing1LimitDegrees=45
  swing2LimitDegrees=45
  twistLimitDegrees=45
  twistMotion=Locked
 }
 {
  bone1Name=dyn_strand_front_02
  bone2Name=dyn_strand_front_01
  swing1LimitDegrees=45
  swing2LimitDegrees=45
  twistLimitDegrees=45
  twistMotion=Locked
 }
 {
  bone1Name=dyn_strand_front_03
  bone2Name=dyn_strand_front_02
  swing1LimitDegrees=45
  swing2LimitDegrees=45
  twistLimitDegrees=45
  twistMotion=Locked
 }
 {
  bone1Name=dyn_strand_front_04
  bone2Name=dyn_hair_mstr
  swing1LimitDegrees=20
  swing2LimitDegrees=45
  twistLimitDegrees=45
  twistMotion=Locked
 }
 {
  bone1Name=dyn_strand_front_05
  bone2Name=dyn_strand_front_04
  swing1LimitDegrees=45
  swing2LimitDegrees=45
  twistLimitDegrees=45
  twistMotion=Locked
 }
 {
  bone1Name=dyn_hair_mstr
  bone2Name=spine_05
  swing1LimitDegrees=45
  swing1Motion=Free
  swing2LimitDegrees=45
  swing2Motion=Free
  twistLimitDegrees=45
  twistMotion=Free
 }
]
