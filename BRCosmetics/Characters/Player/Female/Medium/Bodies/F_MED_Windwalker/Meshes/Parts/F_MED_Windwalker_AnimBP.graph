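# Anim Blueprint for the F_MED_Windwalker body part that adds physics to ponytail and hair strands over the pose copied from the attached parent.
# A skydiving bool picks between two ponytail RigidBody sims with different base bones, then sprint/skydive-weighted bone offsets and per-bone AnimDynamics strands follow.
# The EventGraph drives sim-space rotation and gravity through deprecated helpers, using fixed gravity when DBNO-carried.
# Strand and front-hair RigidBody nodes are not wired to the output.
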AnimBlueprint /BRCosmetics/Characters/Player/Female/Medium/Bodies/F_MED_Windwalker/Meshes/Parts/F_MED_Windwalker_AnimBP
target_skeleton=&/F_MED_Windwalker_Skeleton
graphs [
 AnimGraph {nodes [
   n1 = RigidBody(Alpha=0.7) {
    bOverrideWorldGravity=true
    overridePhysicsAsset=&/F_MED_Windwalker_strands_Physics
    simulationSpace=BaseBoneSpace
    alphaBoolBlend {blendOption=Linear}
    baseBoneRef {boneName=root}
    componentLinearAccScale=<1 1 1>
    componentLinearVelScale=<0.5 0.5 0.5>
    overrideWorldGravity {z=-5555}
   }
   RigidBody(ComponentPose=@n1, Alpha=0.3) {
    bOverrideWorldGravity=true
    overridePhysicsAsset=&/F_MED_Windwalker_Front_Hair_Physics
    simulationSpace=BaseBoneSpace
    alphaBoolBlend {blendOption=Linear}
    baseBoneRef {boneName=root}
    componentLinearAccScale=<1 1 1>
    componentLinearVelScale=<0.5 0.5 0.5>
    overrideWorldGravity {z=-5555}
   }
   n2 = "Transform (Modify) Bone - Bone: dyn_root_simspace"(
    ComponentPose=Local To Component(LocalPose=Copy Pose From Mesh() {bUseAttachedParent=true})
    Alpha=1
    Translation="0,0,0"
    Rotation=Get sim_space_rotation() -> (sim_space_rotation:Rotator)
    Scale="1,1,1"
   ) {
    rotationMode=BMM_Replace
    rotationSpace=BCS_WorldSpace
    translationSpace=BCS_BoneSpace
    alphaBoolBlend {blendOption=Linear}
    alphaScaleBiasClamp {bInterpResult=true, interpSpeedDecreasing=2, interpSpeedIncreasing=2}
    boneToModify {boneName=dyn_root_simspace}
   }
   n3 = RigidBody!(ComponentPose=@n2, Alpha=0.8, OverrideWorldGravity:Vector=Get Gravity Override(), ExternalForce) -> (Pose) {
    bClampLinearTranslationLimitToRefPose=true
    bOverrideWorldGravity=true
    overridePhysicsAsset=&/F_MED_Windwalker_Ponytail_Physics
    simulationSpace=BaseBoneSpace
    alphaBoolBlend {blendOption=Linear}
    baseBoneRef {boneName=spine_05}
    componentAppliedLinearAccClamp=<3000 3000 3000>
    componentLinearAccScale=<0.4 0.4 0.6>
    componentLinearVelScale=<0.3 0.3 0.3>
    overrideWorldGravity {z=-850}
   }
   n4 = Component To Local(ComponentPose=@n3)
   n5 = "Transform (Modify) Bone - Bone: dyn_root_simspace"(
    ComponentPose=Local To Component(LocalPose=Copy Pose From Mesh() {bUseAttachedParent=true})
    Alpha=1
    Translation="0,0,0"
    Rotation=Get sim_space_rotation()
    Scale="1,1,1"
   ) {
    rotationMode=BMM_Replace
    rotationSpace=BCS_WorldSpace
    translationSpace=BCS_BoneSpace
    alphaBoolBlend {blendOption=Linear}
    alphaScaleBiasClamp {bInterpResult=true, interpSpeedDecreasing=2, interpSpeedIncreasing=2}
    boneToModify {boneName=dyn_root_simspace}
   }
   n6 = RigidBody(ComponentPose=@n5, Alpha=0.8, OverrideWorldGravity=Get Gravity Override()) {
    bClampLinearTranslationLimitToRefPose=true
    bOverrideWorldGravity=true
    overridePhysicsAsset=&F_MED_Windwalker_Ponytail_Physics
    simulationSpace=BaseBoneSpace
    alphaBoolBlend {blendOption=Linear}
    baseBoneRef {boneName=head}
    componentAppliedLinearAccClamp=<3000 3000 3000>
    componentLinearAccScale=<0.2 0.2 0.3>
    componentLinearVelScale=<0.15 0.15 0.15>
    overrideWorldGravity {z=-850}
   }
   n7 = Component To Local(ComponentPose=@n6)
   n8 = Get bIsSkyDiving()
   n9 = "Transform (Modify) Bone - Bone: dyn_ponytail_01"(
    ComponentPose=Local To Component(LocalPose=Blend Poses by bool(BlendPose_0=@n7, BlendPose_1=@n4, bActiveValue=@n8) {blendType=Linear})
    Alpha=Get NewVar_0() -> (NewVar_0:double)
    Translation="0,0,0"
    Rotation="0,-20,0"
    Scale="1,1,1"
   ) {
    rotationMode=BMM_Additive
    rotationSpace=BCS_BoneSpace
    alphaBoolBlend {blendOption=Linear}
    alphaScaleBiasClamp {bInterpResult=true, interpSpeedDecreasing=4, interpSpeedIncreasing=4}
    boneToModify {boneName=dyn_ponytail_01}
    rotation {yaw=-20}
   }
   n10 = "Transform (Modify) Bone - Bone: dyn_ponytail_02"(ComponentPose=@n9, Alpha=Get NewVar_0(), Translation="0,0,0", Rotation="0,-20,0", Scale="1,1,1") {
    rotationMode=BMM_Additive
    rotationSpace=BCS_BoneSpace
    alphaBoolBlend {blendOption=Linear}
    alphaScaleBiasClamp {bInterpResult=true, interpSpeedDecreasing=4, interpSpeedIncreasing=4}
    boneToModify {boneName=dyn_ponytail_02}
    rotation {yaw=-20}
   }
   n11 = "Transform (Modify) Bone - Bone: dyn_ponytail_03"(ComponentPose=@n10, Alpha=Get NewVar_0(), Translation="0,0,0", Rotation="0,-25,0", Scale="1,1,1") {
    rotationMode=BMM_Additive
    rotationSpace=BCS_BoneSpace
    alphaBoolBlend {blendOption=Linear}
    alphaScaleBiasClamp {bInterpResult=true, interpSpeedDecreasing=4, interpSpeedIncreasing=4}
    boneToModify {boneName=dyn_ponytail_03}
    rotation {yaw=-25}
   }
   n12 = Anim Dynamics - Bone: dyn_strand_l_01!(ComponentPose=@n11, ExternalForce="0,0,0") -> (Pose) {
    alpha=0.5
    angularDampingOverride=0.99
    angularSpringConstant=100
    bAngularSpring=true
    bGravityOverrideInSimSpace=true
    bOverrideAngularDamping=true
    gravityScale=0.35
    simulationSpace=RootRelative
    alphaBoolBlend {blendOption=Linear}
    boundBone {boneName=dyn_strand_l_01}
    componentAppliedLinearAccClamp=<3000 3000 3000>
    componentLinearAccScale {x=0.5, z=0.6}
    componentLinearVelScale {x=0.5, z=0.6}
    physicsBodyDefinitions [{
      sphereCollisionRadius=0
      boundBone {boneName=dyn_strand_l_01}
      boxExtents=<8 30 30>
      constraintSetup {angularLimitsMax=<0 0 30>, angularLimitsMin=<0 -20 -5>, angularTarget=<1 0 0>}
      localJointOffset=<6 0 0>
     }]
   }
   n13 = Anim Dynamics - Bone: dyn_strand_l_02!(ComponentPose=@n12, ExternalForce="0,0,0") -> (Pose) {
    alpha=0.5
    angularDampingOverride=0.99
    angularSpringConstant=100
    bAngularSpring=true
    bGravityOverrideInSimSpace=true
    bOverrideAngularDamping=true
    gravityScale=0.35
    simulationSpace=RootRelative
    alphaBoolBlend {blendOption=Linear}
    boundBone {boneName=dyn_strand_l_02}
    componentAppliedLinearAccClamp=<3000 3000 3000>
    componentLinearAccScale {x=0.5, z=0.6}
    componentLinearVelScale {x=0.5, z=0.6}
    physicsBodyDefinitions [{
      sphereCollisionRadius=0
      boundBone {boneName=dyn_strand_l_02}
      boxExtents=<8 30 30>
      constraintSetup {angularLimitsMax=<0 0 0>, angularLimitsMin=<0 -20 -60>, angularTarget=<1 0 0>}
      localJointOffset=<6 0 0>
     }]
   }
   n14 = Anim Dynamics - Bone: dyn_strand_l_03!(ComponentPose=@n13, ExternalForce="0,0,0") -> (Pose) {
    alpha=0.5
    angularDampingOverride=0.99
    angularSpringConstant=100
    bAngularSpring=true
    bGravityOverrideInSimSpace=true
    bOverrideAngularDamping=true
    gravityScale=0.35
    simulationSpace=RootRelative
    alphaBoolBlend {blendOption=Linear}
    boundBone {boneName=dyn_strand_l_03}
    componentAppliedLinearAccClamp=<3000 3000 3000>
    componentLinearAccScale {x=0.5, z=0.6}
    componentLinearVelScale {x=0.5, z=0.6}
    physicsBodyDefinitions [{
      sphereCollisionRadius=0
      boundBone {boneName=dyn_strand_l_03}
      boxExtents=<8 30 30>
      constraintSetup {angularLimitsMax=<0 0 20>, angularLimitsMin=<0 -20 -60>, angularTarget=<1 0 0>}
      localJointOffset=<6 0 0>
     }]
   }
   n15 = Anim Dynamics - Bone: dyn_strand_r_01!(ComponentPose=@n14, ExternalForce="0,0,0") -> (Pose) {
    alpha=0.5
    angularDampingOverride=0.995
    angularSpringConstant=100
    bAngularSpring=true
    bGravityOverrideInSimSpace=true
    bOverrideAngularDamping=true
    gravityScale=0.35
    simulationSpace=RootRelative
    alphaBoolBlend {blendOption=Linear}
    boundBone {boneName=dyn_strand_r_01}
    componentAppliedLinearAccClamp=<3000 3000 3000>
    componentLinearAccScale {x=0.5, z=0.6}
    componentLinearVelScale {x=0.5, z=0.6}
    physicsBodyDefinitions [{
      sphereCollisionRadius=0
      boundBone {boneName=dyn_strand_r_01}
      boxExtents=<8 30 30>
      constraintSetup {angularLimitsMax=<0 0 15>, angularLimitsMin=<0 -20 -30>, angularTarget=<1 0 0>}
      localJointOffset=<6 0 0>
     }]
   }
   n16 = Anim Dynamics - Bone: dyn_strand_r_04!(ComponentPose=@n15, ExternalForce="0,0,0") -> (Pose) {
    alpha=0.5
    angularDampingOverride=0.995
    angularSpringConstant=100
    bAngularSpring=true
    bGravityOverrideInSimSpace=true
    bOverrideAngularDamping=true
    gravityScale=0.35
    simulationSpace=RootRelative
    alphaBoolBlend {blendOption=Linear}
    boundBone {boneName=dyn_strand_r_04}
    componentAppliedLinearAccClamp=<3000 3000 3000>
    componentLinearAccScale {x=0.5, z=0.6}
    componentLinearVelScale {x=0.5, z=0.6}
    physicsBodyDefinitions [{
      sphereCollisionRadius=0
      boundBone {boneName=dyn_strand_r_04}
      boxExtents=<8 30 30>
      constraintSetup {angularLimitsMax=<0 0 10>, angularLimitsMin=<0 -20 -60>, angularTarget=<1 0 0>}
      localJointOffset=<6 0 0>
     }]
   }
   n17 = Anim Dynamics - Bone: dyn_strand_r_02!(ComponentPose=@n16, ExternalForce="0,0,0") -> (Pose) {
    alpha=0.5
    angularDampingOverride=0.995
    angularSpringConstant=100
    bAngularSpring=true
    bGravityOverrideInSimSpace=true
    bOverrideAngularDamping=true
    gravityScale=0.35
    simulationSpace=RootRelative
    alphaBoolBlend {blendOption=Linear}
    boundBone {boneName=dyn_strand_r_02}
    componentAppliedLinearAccClamp=<3000 3000 3000>
    componentLinearAccScale {x=0.5, z=0.6}
    componentLinearVelScale {x=0.5, z=0.6}
    physicsBodyDefinitions [{
      sphereCollisionRadius=0
      boundBone {boneName=dyn_strand_r_02}
      boxExtents=<8 30 30>
      constraintSetup {angularLimitsMax=<0 0 10>, angularLimitsMin=<0 -20 -60>, angularTarget=<1 0 0>}
      localJointOffset=<6 0 0>
     }]
   }
   Output Pose(Result=Component To Local(ComponentPose=@n17)) {layerGroup=None}
  ]}
 EventGraph {nodes [
   Event BlueprintUpdateAnimation()
   > n1 = Sequence!(execute) -> (then_0, then_1, then_2:Exec)
   > n2 = ComplexSimSpaceFunction DEPRECATED!(
    execute
    CharacterPartAnimInstance:Custom Character Part Anim Instance Object Reference=Self-Reference()
    InterpSpeedMax:double=9
    PawnSpeedMax:double=20
    RInterpErrorTolerance:double=1
    DeltaTime:double=0
    DebugDraw:bool=false
    sim_space_rotation_in:Rotator=Get sim_space_rotation()
    inComponentSpaceIn:bool=Get InComponentSpace() -> (InComponentSpace:bool)
    inComponentSpaceLastFrameIn:bool=Get InComponentSpaceLastFrame() -> (InComponentSpaceLastFrame:bool)
    lockedOnSimSpaceIn:bool=Get LockedOnSimSpace() -> (LockedOnSimSpace:bool)
    InterpSpeedIn:double=Get InterpSpeed() -> (InterpSpeed:double)
   ) -> (
    then
    sim_space_rotation:Rotator
    inComponentSpaceOut:bool
    inComponentSpaceLastFrameOut:bool
    lockedOnSimSpaceOut:bool
    InterpSpeedOut:double
   )
   > Set sim_space_rotation(sim_space_rotation:Rotator=@n2) -> (Output_Get:Rotator)
   > Set InComponentSpace(InComponentSpace:bool=@n2.inComponentSpaceOut) -> (Output_Get:bool)
   > Set InComponentSpaceLastFrame(InComponentSpaceLastFrame:bool=@n2.inComponentSpaceLastFrameOut) -> (Output_Get:bool)
   > Set LockedOnSimSpace(LockedOnSimSpace:bool=@n2.lockedOnSimSpaceOut) -> (Output_Get:bool)
   > Set InterpSpeed(InterpSpeed:double=@n2.InterpSpeedOut) -> (Output_Get:double)
   n3 = Get bIsSkyDiving()
   n4 = Get bIsSprinting()
   Set NewVar_0(execute=@n1.then_2, NewVar_0:double=SelectFloat(A=0.5, B=SelectFloat(A=1, B=0, bPickA=@n3), bPickA=@n4)) -> (Output_Get:double)
   n5 = Gravity Override State Function DEPRECATED(
    execute=@n1.then_1
    CharacterPartAnimInstance=Self-Reference()
    GameTimeInSeconds=GetGameTimeInSeconds()
    GravityOverrideParams=Get Gravity Override Params()
   )
   > n6 = Branch(Condition=Get bIsDBNOCarried())
   > Set Gravity Override(Gravity Override="0,0,-980")
   Set Gravity Override(execute=@n6.else, Gravity Override=@n5)
   AND Boolean(A=@n4, B=NOT Boolean(A=Get bIsCrouching()))
  ]}
]
variables [
 Gravity Override {type=Vector}
 Gravity Override Params {type=GravityOverrideParamsStruct}
 InComponentSpace {type=boolean}
 InComponentSpaceLastFrame {type=boolean}
 InterpSpeed {type=number}
 LockedOnSimSpace {type=boolean}
 NewVar_0 {type=number}
 sim_space_rotation {type=Rotator}
]
